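# Niagara burst effect for the Cloaked Assassin female pickaxe: purple lightning, glowing runes, orbiting mesh squares and bone-following sprites on the weapon's skeletal mesh.
# Two "EndPicker" emitters sample bone or socket pairs, which ribbon emitters read via particle attribute readers to draw curved, jittered arcs between them.
# The smaller arcs' endpoint count and endpoint brightness scale with User.DamageLevel.
# Squares pick one of several mesh renderers through visibility tags and orbit in bone space.
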NiagaraSystem /BRCosmetics/Weapons/FORT_Melee/Pickaxe_Cloaked_Assassin_Female/Effects/NS_Pickaxe_CloakedAssassin_Burst
dependencies {
 usedDataInterfaces [&NiagaraDataInterfaceColorCurve, &NiagaraDataInterfaceParticleRead, &NiagaraDataInterfaceCurve]
 usedDynamicInputs []
 usedModules [&SystemState]
}
emitters {
 BigLightningEndPicker {
  emitterUpdateScript [
   &EmitterState()
   &SpawnBurst_Instantaneous(Spawn Count=1, Use Spawn Probability=false)
  ]
  particleSpawnScript [
   /Niagara/Modules/Spawn/Initialization/InitializeParticle(
    Color=&UniformRangedLinearColor(Minimum=<rgba 0.2 0 1 1>, Maximum=<rgba 2 0 3 1>, Color Channel Mode=::NewEnumerator2)
    Lifetime=&UniformRangedFloat(Minimum=0.2, Maximum=0.3)
    Module.Write Color=true
    Module.Write Lifetime=true
    Module.Write Mass=true
    Module.Write Material Random=true
    Module.Write Position=false
    Module.Write Scale=false
    Module.Write SpriteRotation=true
    Module.Write SpriteSize=true
    Position=&SimulationPosition()
    Sprite Rotation=90
    Sprite Size=&Vector2DFromFloat(Value=5)
   )
   &SkeletalMeshLocation(
    Bone / Socket Index=&MakeIntFromFloat(Float=0)
    Inherit Velocity=false
    MaskSpawnGroup=false
    Mesh Sampling Type=::NewEnumerator5
    Skeletal Mesh=$User.SkeletalMesh
    Socket Sampling Mode=::NewEnumerator5
   ) {module=SkeletalMeshLocation001}
   /SetVariables(Particles.HomePosition:Vector3f=$Output.SkeletalMeshLocation001.SampledPosition) {isSetParameters=true}
   &SkeletalMeshLocation(
    Apply To Particle Position=false
    Bone / Socket Index=&UniformRangedInt(Minimum=1, Maximum=2)
    Inherit Velocity=false
    MaskSpawnGroup=false
    Mesh Sampling Type=::NewEnumerator5
    Skeletal Mesh=$User.SkeletalMesh
    Socket Sampling Mode=::NewEnumerator5
   )
   /SetVariables(Particles.FriendPosition:Vector3f=$Output.SkeletalMeshLocation.SampledPosition) {isSetParameters=true, module=SetVariables_#1}
  ]
  particleUpdateScript [
   &SolveForcesAndVelocity(Clamp Velocity=false, Limit Acceleration=false)
   &ParticleState()
  ]
  rendererClasses [&NiagaraSpriteRendererProperties]
  renderers [
   {rendererClass=&NiagaraSpriteRendererProperties}
   {rendererClass=&NiagaraSpriteRendererProperties, rendererIndex=1}
  ]
 }
 BigLightningRibbon {
  emitterSpawnScript [/Game/Effects/Niagara/Modules/Attributes/FN_Source_v2(AttributeReader=DataInterface() {propertyValues {EmitterBinding {bindingMode=Other, emitterName=BigLightningEndPicker}}})]
  emitterUpdateScript [
   &EmitterState()
   /SetVariables(Emitter.ParticlesPerRibbon:NiagaraFloat=&UniformRangedFloat(Minimum=30, Maximum=50)) {isSetParameters=true, module=SetVariables_#2}
   &SpawnRate(
    SpawnRate=/Game/Effects/Niagara/DynamicInputs/Attributes/FN_GetEmitterParticleCount(AttributeReader:NiagaraDataInterfaceParticleRead=$Emitter.AttributeReader, Multiplier:NiagaraFloat=$Emitter.ParticlesPerRibbon)
    Use Spawn Probability=false
   )
  ]
  particleSpawnScript [
   &self:GetInfoFromParentAbridged_02() {module=GetInfoFromParentAbridged_02}
   /SetVariables(Particles.AgeByParent:NiagaraFloat=&FloatFromCurve(FloatCurve=DataInterface() {propertyValues {Curve {keys [<at 0.2 1>,<at 0.8 0>]}}}, CurveIndex=$Particles.AgeImported)) {isSetParameters=true}
   &InitializeRibbon(
    Color=<rgba 2.4952 0 6 1>
    Lifetime=0.3
    Module.Write Color=true
    Module.Write Lifetime=true
    Module.Write Mass=false
    Module.Write Material Random=true
    Module.Write Position=false
    Module.Write RibbonTwist=false
    Module.Write RibbonWidth=true
    Ribbon Width=/Niagara/DynamicInputs/Math/LerpMultipleFloats(
     Float Count:ENiagara_MultipleLerpCount=::NewEnumerator0
     1:NiagaraFloat=2
     2:NiagaraFloat=&UniformRangedFloat(Minimum=25, Maximum=30)
     3:NiagaraFloat=2
     4:NiagaraFloat=?
     5:NiagaraFloat=?
     Alpha:NiagaraFloat=$Particles.AgeByParent
    )
   )
   &InitializeParticle(
    Color=&ColorFromCurve(
     ColorCurve=DataInterface() {propertyValues {
       AlphaCurve {keys [<at 0 1>,<at 1 1>]}
       BlueCurve {keys [<at 0.0171 0>,<at 0.4894 0.0261>,<at 1.0012 1>]}
       GreenCurve {keys [<at 0.0171 0>,<at 0.4894 1>,<at 1.0012 0>]}
       RedCurve {keys [<at 0.0171 1>,<at 0.4894 0>,<at 1.0012 0>]}
      }}
     CurveIndex=$Particles.AgeByParent
    )
    Lifetime=0.7
    Module.Write Color=false
    Module.Write Lifetime=false
    Module.Write Mass=false
    Module.Write Material Random=true
    Module.Write Position=false
    Module.Write Scale=false
    Module.Write SpriteRotation=false
    Module.Write SpriteSize=true
    Sprite Size=&Vector2DFromFloat(Value=&Multiply_Float(A=1, B=1))
   )
   /SetVariables(
    Particles.OffsetPerRibbon:Vector3f=&UniformRangedVector(Minimum=<-15 -30 -1>, Maximum=<10 60 1>, Randomness Mode=::NewEnumerator1, Random Seed=$Particles.ParentIDInt, Fixed Random Seed=true)
    Particles.OffsetRandom:Vector3f=&Multiply_VectorByFloat(Vector=&UniformRangedVector(Minimum=<-1 -1 -0.2>, Maximum=<1 1 0.2>), Float=4)
   ) {isSetParameters=true, module=SetVariables_#3}
   /SetVariables(Particles.Position=&Add_Vector(
     A=&Lerp_Vector(A=$Particles.ParentPosition, B=$Particles.ParentFriendPosition, Alpha=$Particles.AgeByParent)
     B=&Multiply_VectorByFloat(
      Vector=&Add_Vector(A=$Particles.OffsetPerRibbon, B=$Particles.OffsetRandom)
      Float=&FloatFromCurve(
       FloatCurve=DataInterface() {propertyValues {Curve {keys [
           {arriveTangent=0.0416, interpMode=RCIM_Cubic, leaveTangent=0.0416, tangentMode=RCTM_User, time=0.1, value=0}
           {interpMode=RCIM_Cubic, tangentMode=RCTM_User, time=0.4951, value=1}
           {arriveTangent=-0.0664, interpMode=RCIM_Cubic, leaveTangent=-0.0664, tangentMode=RCTM_User, time=0.9, value=0}
          ]}}}
       CurveIndex=$Particles.AgeByParent
      )
     )
    )) {isSetParameters=true, module=SetVariables_#4}
   &AddVelocity(Velocity=&Multiply_VectorByFloat(
     Vector=$Particles.OffsetRandom
     Float=&FloatFromCurve(
      FloatCurve=DataInterface() {propertyValues {Curve {keys [
          {arriveTangent=0.0416, interpMode=RCIM_Cubic, leaveTangent=0.0416, tangentMode=RCTM_User, time=0, value=0}
          {interpMode=RCIM_Cubic, tangentMode=RCTM_User, time=0.5, value=1}
          {arriveTangent=-0.0664, interpMode=RCIM_Cubic, leaveTangent=-0.0664, tangentMode=RCTM_User, time=1, value=0}
         ]}}}
      CurveIndex=$Particles.AgeByParent
      Scale Curve=10
     )
    ))
  ]
  particleUpdateScript [
   &ParticleState()
   &Drag(Drag=5, Use Linear Drag=true, Use Rotational Drag=false)
   &SolveForcesAndVelocity(Clamp Velocity=false, Limit Acceleration=false)
   &ScaleRibbonWidth(Ribbon Width Scale=&FloatFromCurve(FloatCurve=DataInterface() {propertyValues {Curve {keys [
         {arriveTangent=7.8756, interpMode=RCIM_Cubic, leaveTangent=7.8756, tangentMode=RCTM_User, time=0, value=0}
         {interpMode=RCIM_Cubic, tangentMode=RCTM_User, time=0.2369, value=1}
         {arriveTangent=-2.4215, interpMode=RCIM_Cubic, leaveTangent=-2.4215, tangentMode=RCTM_User, time=1, value=0}
        ]}}}))
   &ScaleColor(
    Scale RGB=&VectorFromFloat(Value=&LerpMultipleFloats(
      Float Count:ENiagara_MultipleLerpCount=::NewEnumerator2
      1:NiagaraFloat=3
      2:NiagaraFloat=1
      3:NiagaraFloat=1
      4:NiagaraFloat=1
      5:NiagaraFloat=0.1
      Alpha:NiagaraFloat=$Particles.NormalizedAge
     ))
    ScaleA=false
    ScaleRGB=true
   )
  ]
  rendererClasses [&NiagaraRibbonRendererProperties, &NiagaraSpriteRendererProperties]
  renderers [
   {rendererClass=&NiagaraRibbonRendererProperties}
   {rendererClass=&NiagaraSpriteRendererProperties, rendererIndex=1}
  ]
 }
 Decal {
  emitterUpdateScript [
   &EmitterState()
   &SpawnBurst_Instantaneous(Spawn Count=40, Use Spawn Probability=false)
  ]
  particleSpawnScript [
   &InitializeParticle(
    Color=<rgba 2.7561 0.2322 4 1>
    Lifetime=&UniformRangedFloat(Minimum=0.1, Maximum=0.6)
    Module.Write Color=true
    Module.Write Lifetime=true
    Module.Write Mass=true
    Module.Write Material Random=true
    Module.Write Position=false
    Module.Write Scale=false
    Module.Write SpriteRotation=false
    Module.Write SpriteSize=true
    Position=&SimulationPosition()
    Sprite Size=&Vector2DFromFloat(Value=40)
   )
   &DynamicMaterialParameters(
    Index 0 Param 1=0.5
    Index 0 Param 3=5
    Module.Param0WriteEnabled=false
    Module.Param1WriteEnabled=true
    Module.Param2WriteEnabled=true
    Module.Param3WriteEnabled=true
   )
   &CameraOffset(Camera Offset Amount=13)
   /SetVariables(Particles.ParticleRandom:NiagaraFloat=&UniformRangedFloat(Minimum=0, Maximum=100)) {isSetParameters=true}
  ]
  particleUpdateScript [
   &ParticleState()
   &SkeletalMeshLocation(
    Apply Sampled Orientation to Mesh Particles=true
    Bone / Socket Index=&ModuloInt(A=&MakeIntFromFloat(Float=$Particles.ParticleRandom), B=20)
    Bone Sampling Mode=::NewEnumerator6
    Inherit Velocity=false
    Interpolate Mesh Sampling=false
    MaskSpawnGroup=false
    Sampled Position Offset=&Lerp_Vector(
     A=&UniformRangedVector(
      Minimum=<5 5 -7>
      Maximum=<-5 -5 -7>
      Randomness Mode=::NewEnumerator1
      Random Seed=&MakeIntFromFloat(Float=$Particles.ParticleRandom)
      Fixed Random Seed=true
     )
     B=&UniformRangedVector(
      Minimum=<-5 -5 20>
      Maximum=<5 5 50>
      Randomness Mode=::NewEnumerator1
      Random Seed=&MakeIntFromFloat(Float=$Particles.ParticleRandom)
      Fixed Random Seed=true
     )
     Alpha=$Particles.NormalizedAge
    )
    Skeletal Mesh=$User.SkeletalMesh
   )
   &DynamicMaterialParameters(
    Index 0 Param 1=&FloatFromCurve(FloatCurve=DataInterface() {propertyValues {Curve {keys [
         {arriveTangent=0.9726, interpMode=RCIM_Cubic, time=0.1, value=0.1}
         {arriveTangent=12.1538, interpMode=RCIM_Cubic, leaveTangent=12.1538, time=0.45, value=3}
         <at 0.75 8>
        ]}}})
    Module.Param0WriteEnabled=true
    Module.Param1WriteEnabled=false
    Module.Param2WriteEnabled=false
    Module.Param3WriteEnabled=false
   ) {module=DynamicMaterialParameters001}
   &ScaleColor(
    Scale Alpha=&FloatFromCurve(FloatCurve=DataInterface() {propertyValues {Curve {keys [<at 0.75 1>,<at 1 0>]}}})
    ScaleA=true
    ScaleRGB=false
   )
   &ScaleSpriteSize(Scale Factor=&Vector2DFromFloat(Value=&FloatFromCurve(FloatCurve=DataInterface() {propertyValues {Curve {keys [{arriveTangent=-1.4522, interpMode=RCIM_Cubic, leaveTangent=-1.4522, tangentMode=RCTM_User, time=0, value=1}, <at 1 0.5>]}}})))
   &KillParticlesInVolume(Box Size:Vector3f=<20 20 90>, Invert Volume=true, Kill Shape=::NewEnumerator1, Origin Offset:Vector3f=<-3 0 12>)
  ]
  rendererClasses [&NiagaraSpriteRendererProperties]
  renderers [
   {rendererClass=&NiagaraSpriteRendererProperties}
   {rendererClass=&NiagaraSpriteRendererProperties, rendererIndex=1}
  ]
 }
 LightningEndPicker {
  emitterUpdateScript [
   /SetVariables(
    Emitter.SetLoopDelay:NiagaraFloat=&UniformRangedFloat(Minimum=0.1, Maximum=0.9)
    Emitter.SpawnVariable:NiagaraFloat=&ClampFloat(Float=&Multiply_Float(A=0.5, B=&Add_Float(A=&MakeFloatFromInt(Int=$User.DamageLevel), B=-1)))
   ) {isSetParameters=true}
   /SetVariables(
    Emitter.PickSpawnRate:NiagaraInt32=&MakeIntFromFloat(Float=&Multiply_Float(A=5, B=&Multiply_Float(A=$Emitter.SpawnVariable, B=$Emitter.SpawnVariable)))
    Emitter.SetLoopDuration:NiagaraFloat=&Subtract_Float(A=1, B=$Emitter.SetLoopDelay)
   ) {isSetParameters=true}
   &EmitterState()
   &SpawnBurst_Instantaneous(Spawn Count=$Emitter.PickSpawnRate, Use Spawn Probability=false)
  ]
  particleSpawnScript [
   &ApplyInitialForces()
   &InitializeParticle(
    Color=&UniformRangedLinearColor(Minimum=<rgba 0.2 0 1 1>, Maximum=<rgba 2 0 3 1>, Color Channel Mode=::NewEnumerator2)
    Lifetime=0.5
    Module.Write Color=true
    Module.Write Lifetime=true
    Module.Write Mass=true
    Module.Write Material Random=true
    Module.Write Position=false
    Module.Write Scale=false
    Module.Write SpriteRotation=true
    Module.Write SpriteSize=true
    Position=&SimulationPosition()
    Sprite Rotation=90
    Sprite Size=&Vector2DFromFloat(Value=2)
   )
   /SetVariables(Particles.PickBone1:NiagaraInt32=&UniformRangedInt(Minimum=0, Maximum=20)) {isSetParameters=true}
   &SkeletalMeshLocation(
    Apply To Particle Position=false
    Bone / Socket Index=$Particles.PickBone1
    Bone Sampling Mode=::NewEnumerator6
    Inherit Velocity=false
    MaskSpawnGroup=false
    Skeletal Mesh=$User.SkeletalMesh
   )
   /SetVariables(
    Particles.FriendPosition:Vector3f=$Output.SkeletalMeshLocation.SampledPosition
    Particles.PickBone2:NiagaraInt32=&ModuloInt(
     A=&Add_Integer(
      A=$Particles.PickBone1
      B=&MakeIntFromFloat(Float=&UniformAOrBFloat(A=-7, B=7))
     )
     B=20
    )
   ) {isSetParameters=true, module=SetVariables_#1}
   &SkeletalMeshLocation(
    Bone / Socket Index=$Particles.PickBone2
    Bone Sampling Mode=::NewEnumerator6
    Inherit Velocity=false
    MaskSpawnGroup=false
    Skeletal Mesh=$User.SkeletalMesh
   ) {module=SkeletalMeshLocation001}
   &CameraOffset(Camera Offset Amount=2)
   &AddVelocity(Velocity=&UniformRangedVector(Minimum=<-5 -5 -5>, Maximum=<5 5 5>))
   /SetVariables(Particles.LightExponent=1, Particles.LightRadius=&UniformRangedFloat(Minimum=7, Maximum=12)) {isSetParameters=true}
  ]
  particleUpdateScript [
   &ParticleState()
   /SetVariables(Particles.RenormalizedAge:NiagaraFloat=&self:RenormalizeAge_02()) {isSetParameters=true}
   &SolveForcesAndVelocity(Clamp Velocity=false, Limit Acceleration=false)
   /SetVariables(Particles.Position=&self:ClampVector_02(Max:Vector3f=<5 8 55>, Min:Vector3f=<-11 -8 5>, Position:Vector3f=Linked() {linkedVariable {name=Particles.Position, type=NiagaraPosition}})) {isSetParameters=true}
   /SetVariables(Particles.LightRadius=&Multiply_Float(
     A=$Particles.Initial.LightRadius
     B=&ClampFloat(Float=&Lerp_Float(A=2, B=-2, Alpha=$Particles.NormalizedAge))
    )) {isSetParameters=true}
   &ScaleColor(
    Color Value To Scale=&MultiplyLinearColor(
     Linear Color=$Particles.Initial.Color
     Scale Factor=&MakeVector4_FromFloat(Float=&Multiply_Float(
       A=0.5
       B=&ClampFloat(Float=&Add_Float(A=&MakeFloatFromInt(Int=$User.DamageLevel), B=-2))
      ))
    )
    Scale Mode=::NewEnumerator1
    Scale RGBA=&Vector4FromCurve(Vector4Curve=DataInterface() {propertyValues {
       AlphaCurve {keys [<at 0 1>,<at 1 1>]}
       BlueCurve {keys [<at 0 3>,<at 0.1619 1>,<at 0.5 0.1>]}
       GreenCurve {keys [<at 0 3>,<at 0.1619 1>,<at 0.5 0>]}
       RedCurve {keys [<at 0 3>,<at 0.1619 1>,<at 0.5 0>]}
      }})
    ScaleRGBA=true
   )
  ]
  rendererClasses [&NiagaraSpriteRendererProperties]
  renderers [{rendererClass=&NiagaraSpriteRendererProperties}]
 }
 LightningRibbon {
  emitterSpawnScript [&FN_Source_v2(AttributeReader=DataInterface() {propertyValues {EmitterBinding {bindingMode=Other, emitterName=LightningEndPicker}}})]
  emitterUpdateScript [
   &EmitterState()
   /SetVariables(Emitter.ParticlesPerRibbon=25) {isSetParameters=true, module=SetVariables_#2}
   &SpawnRate(
    SpawnRate=&FN_GetEmitterParticleCount(AttributeReader:NiagaraDataInterfaceParticleRead=$Emitter.AttributeReader, Multiplier:NiagaraFloat=$Emitter.ParticlesPerRibbon)
    Use Spawn Probability=false
   )
  ]
  particleSpawnScript [
   &self:GetInfoFromParent_02() {module=GetInfoFromParent_02}
   /SetVariables(Particles.DistanceBetweenParents:NiagaraFloat=&ClampFloat(Float=&Add_Float(
      A=0.46
      B=&Multiply_Float(
       A=0.035
       B=/Niagara/DynamicInputs/TypeConversions/VectorLengthSafe(Vector:Vector3f=&Subtract_Vector(A=$Particles.ParentPosition, B=$Particles.ParentFriendPosition))
      )
     ))) {isSetParameters=true}
   &InitializeRibbon(
    Color=<rgba 2.4952 0 6 1>
    Module.Write Color=true
    Module.Write Lifetime=true
    Module.Write Mass=true
    Module.Write Material Random=true
    Module.Write Position=false
    Module.Write RibbonTwist=false
    Module.Write RibbonWidth=true
    Ribbon Width=&LerpMultipleFloats(
     Float Count:ENiagara_MultipleLerpCount=::NewEnumerator0
     1:NiagaraFloat=0.5
     2:NiagaraFloat=6
     3:NiagaraFloat=0.5
     4:NiagaraFloat=?
     5:NiagaraFloat=?
     Alpha:NiagaraFloat=$Particles.AgeByParent
    )
   )
   &InitializeParticle(
    Color=&ColorFromCurve(
     ColorCurve=DataInterface() {propertyValues {
       AlphaCurve {keys [<at 0 1>,<at 1 1>]}
       BlueCurve {keys [<at 0.0171 0>,<at 0.4894 0.0261>,<at 1.0012 1>]}
       GreenCurve {keys [<at 0.0171 0>,<at 0.4894 1>,<at 1.0012 0>]}
       RedCurve {keys [<at 0.0171 1>,<at 0.4894 0>,<at 1.0012 0>]}
      }}
     CurveIndex=$Particles.AgeByParent
    )
    Lifetime=0.7
    Module.Write Color=false
    Module.Write Lifetime=false
    Module.Write Mass=true
    Module.Write Material Random=true
    Module.Write Position=false
    Module.Write Scale=false
    Module.Write SpriteRotation=false
    Module.Write SpriteSize=true
    Sprite Size=&Vector2DFromFloat(Value=&Multiply_Float(A=1, B=1))
   )
   /SetVariables(Particles.Position=&Lerp_Vector(A=$Particles.ParentPosition, B=$Particles.ParentFriendPosition, Alpha=$Particles.AgeByParent)) {isSetParameters=true, module=SetVariables_#4}
   /SetVariables(
    Particles.VelocityPerRibbon:Vector3f=&UniformRangedVector(Minimum=<-20 -20 -20>, Maximum=<20 20 20>, Randomness Mode=::NewEnumerator1, Random Seed=$Particles.ParentIDInt, Fixed Random Seed=true)
    Particles.VelocityRandom:Vector3f=&UniformRangedVector(Minimum=<-20 -20 -20>, Maximum=<20 20 20>)
   ) {isSetParameters=true, module=SetVariables_#3}
   &AddVelocity(Velocity=&Multiply_VectorByFloat(
     Vector=&Add_Vector(A=$Particles.VelocityPerRibbon, B=$Particles.VelocityRandom)
     Float=&FloatFromCurve(
      FloatCurve=DataInterface() {propertyValues {Curve {keys [
          {arriveTangent=5.1547, interpMode=RCIM_Cubic, leaveTangent=5.1547, tangentMode=RCTM_User, time=0.2, value=0}
          {interpMode=RCIM_Cubic, tangentMode=RCTM_User, time=0.4951, value=1}
          {arriveTangent=-5.4931, interpMode=RCIM_Cubic, leaveTangent=-5.4931, tangentMode=RCTM_User, time=0.8, value=0}
         ]}}}
      CurveIndex=$Particles.AgeByParent
      Scale Curve=$Particles.DistanceBetweenParents
     )
    ))
   &CameraOffset(Camera Offset Amount=10)
  ]
  particleUpdateScript [
   &ParticleState()
   /Game/Effects/Niagara/Modules/Attributes/FN_UpdateSourcedRibbbons(Inherit Parent Age=false, Locked Particles Position:Vector3f=$Particles.ParentPosition) {module=FN_UpdateLockedParticles}
   &Drag(Drag=2, Use Linear Drag=true, Use Rotational Drag=false)
   &SolveForcesAndVelocity(Clamp Velocity=false, Limit Acceleration=false)
   &ScaleRibbonWidth(Ribbon Width Scale=&LerpMultipleFloats(
     Float Count:ENiagara_MultipleLerpCount=::NewEnumerator2
     1:NiagaraFloat=0
     2:NiagaraFloat=1
     3:NiagaraFloat=1
     4:NiagaraFloat=0
     5:NiagaraFloat=0
     Alpha:NiagaraFloat=$Particles.NormalizedAge
    ))
   &ScaleColor(
    Scale RGB=&VectorFromFloat(Value=&LerpMultipleFloats(
      Float Count:ENiagara_MultipleLerpCount=::NewEnumerator2
      1:NiagaraFloat=3
      2:NiagaraFloat=1
      3:NiagaraFloat=1
      4:NiagaraFloat=1
      5:NiagaraFloat=0
      Alpha:NiagaraFloat=$Particles.NormalizedAge
     ))
    ScaleA=false
    ScaleRGB=true
   )
  ]
  rendererClasses [&NiagaraRibbonRendererProperties, &NiagaraSpriteRendererProperties]
  renderers [
   {rendererClass=&NiagaraRibbonRendererProperties}
   {rendererClass=&NiagaraSpriteRendererProperties, rendererIndex=1}
  ]
 }
 LightningSprites {
  emitterUpdateScript [
   &EmitterState()
   &SpawnBurst_Instantaneous(Spawn Count=20, Use Spawn Probability=false)
  ]
  particleSpawnScript [
   &InitializeParticle(
    Color=<rgba 0.8317 0 2 1>
    Lifetime=&UniformRangedFloat(Minimum=0.25, Maximum=1)
    Module.Write Color=true
    Module.Write Lifetime=true
    Module.Write Mass=true
    Module.Write Material Random=true
    Module.Write Position=false
    Module.Write Scale=false
    Module.Write SpriteRotation=true
    Module.Write SpriteSize=true
    Position=&SimulationPosition()
    Sprite Rotation=&UniformRangedFloat(Minimum=0, Maximum=360)
    Sprite Size=&Vector2DFromFloat(Value=&UniformRangedFloat(Minimum=7, Maximum=9))
   )
   &DynamicMaterialParameters(
    Index 0 Param 1=&UniformRangedFloat(Minimum=0, Maximum=1)
    Index 0 Param 2=&UniformRangedFloat(Minimum=0.1, Maximum=1)
    Index 0 Param 3=1
    Module.Param0WriteEnabled=true
    Module.Param1WriteEnabled=true
    Module.Param2WriteEnabled=true
    Module.Param3WriteEnabled=false
   ) {module=DynamicMaterialParameters001}
   /SetVariables(
    Particles.FacingAdjustment:Vector3f=&UniformRangedVector(Minimum=<-1 -1 -1>, Maximum=<1 1 1>)
    Particles.SpriteFacing=&UniformRangedVector(Minimum=<-5 -5 -5>, Maximum=<5 5 5>)
    Particles.StableRandom:NiagaraInt32=&UniformRangedInt(Minimum=0, Maximum=5)
    Particles.SubImageIndex=&UniformRangedFloat(Minimum=0, Maximum=3)
   ) {isSetParameters=true}
  ]
  particleUpdateScript [
   &ParticleState()
   &SkeletalMeshLocation(
    Bone / Socket Index=$Particles.StableRandom
    Bone Sampling Mode=::NewEnumerator4
    Inherit Velocity=false
    MaskSpawnGroup=false
    Sampled Position Offset Coordinate Space:ENiagaraMeshTransforms=::NewEnumerator2
    Skeletal Mesh=$User.SkeletalMesh
   )
   &DynamicMaterialParameters(
    Index 0 Param 3=&FloatFromCurve(FloatCurve=DataInterface() {propertyValues {Curve {keys [
         <at 0 1 RCIM_Cubic>
         <at 0.2 0.025 RCIM_Cubic>
         {arriveTangent=2.9631, interpMode=RCIM_Cubic, leaveTangent=2.9631, tangentMode=RCTM_User, time=1, value=1}
        ]}}})
    Module.Param0WriteEnabled=false
    Module.Param1WriteEnabled=false
    Module.Param2WriteEnabled=true
    Module.Param3WriteEnabled=false
   )
   &ScaleColor(
    Scale RGB=&Multiply_VectorByFloat(
     Vector=&MakeVectorFromLinearColorRGB(Color=&ColorFromCurve(ColorCurve=DataInterface() {propertyValues {
         AlphaCurve {keys []}
         BlueCurve {keys [<at 0 1>, {arriveTangent=1, interpMode=RCIM_Cubic, leaveTangent=1, time=0.15, value=10}, <at 1 2 RCIM_Cubic>]}
         GreenCurve {keys [<at 0 1>, {arriveTangent=0.188, interpMode=RCIM_Cubic, leaveTangent=0.188, time=0.15, value=9.045}, <at 1 1.188 RCIM_Cubic>]}
         RedCurve {keys [<at 0 1>, {arriveTangent=0.1927, interpMode=RCIM_Cubic, leaveTangent=0.1927, time=0.15, value=8.5417}, <at 1 1.1927 RCIM_Cubic>]}
        }}))
     Float=1
    )
    ScaleA=false
    ScaleRGB=true
   )
   &ScaleSpriteSize(Scale Factor=&Vector2DFromFloat(Value=&LerpMultipleFloats(
      Float Count:ENiagara_MultipleLerpCount=::NewEnumerator0
      1:NiagaraFloat=0.5
      2:NiagaraFloat=1
      3:NiagaraFloat=0
      4:NiagaraFloat=?
      5:NiagaraFloat=?
      Alpha:NiagaraFloat=$Particles.NormalizedAge
     )))
   /Niagara/Modules/Update/SubUV/SubUVAnimation(SubUV Lookup Index=&Add_Float(A=&Multiply_Float(A=2, B=$Particles.NormalizedAge), B=$Particles.Initial.SubImageIndex))
   /SetVariables(Particles.SpriteFacing=&Add_Vector(
     A=$Particles.SpriteFacing
     B=&Multiply_VectorByFloat(Vector=$Particles.FacingAdjustment, Float=$Particles.NormalizedAge)
    )) {isSetParameters=true}
  ]
  rendererClasses [&NiagaraSpriteRendererProperties]
  renderers [{rendererClass=&NiagaraSpriteRendererProperties}]
 }
 Runes {
  emitterUpdateScript [
   &EmitterState()
   &SpawnBurst_Instantaneous(Spawn Count=&UniformRangedInt(Minimum=12, Maximum=16), Use Spawn Probability=false)
  ]
  particleSpawnScript [
   &InitializeParticle(
    Color=&MakeLinearColor(
     R:NiagaraFloat=1
     G:NiagaraFloat=&UniformRangedFloat(Minimum=1, Maximum=5)
     B:NiagaraFloat=&UniformRangedFloat(Minimum=1, Maximum=10)
     A:NiagaraFloat=&UniformRangedFloat(Minimum=0.5, Maximum=1)
    )
    Lifetime=&UniformRangedFloat(Minimum=0.05, Maximum=&UniformRangedFloat(Minimum=0.2, Maximum=0.6))
    Module.Write Color=true
    Module.Write Lifetime=true
    Module.Write Mass=false
    Module.Write Material Random=true
    Module.Write Position=false
    Module.Write Scale=false
    Module.Write SpriteRotation=true
    Module.Write SpriteSize=true
    Position=&SimulationPosition()
    Sprite Rotation=&UniformRangedFloat(Minimum=0, Maximum=360)
    Sprite Size=&Vector2DFromFloat(Value=&UniformRangedFloat(Minimum=3, Maximum=7))
   )
   /SetVariables(Particles.SpriteInitRotationRate:NiagaraFloat=&Multiply_Float(
     A=&UniformAOrBFloat(A=-1)
     B=&UniformRangedFloat(Minimum=5, Maximum=10)
    )) {isSetParameters=true}
   &SphereLocation(
    MaskSpawnGroup=false
    Normalize Thickness=false
    Offset=<0 0 40>
    Override Seed=false
    Sphere Radius=10
    Surface Only=true
    Surface Only Band Thickness=2
    UseNonUniformScale=false
    UseOffset=true
    UseSphereOrientation=false
   )
   &AddVelocity(Velocity=&Multiply_VectorByFloat(Vector=$Output.SphereLocation.SphereNormal, Float=&UniformRangedFloat(Minimum=100, Maximum=300)))
   &SubUVAnimation(SubUV Animation Mode=::NewEnumerator1)
  ]
  particleUpdateScript [
   &ParticleState()
   &SpriteRotationRate(Rotation Rate=&Multiply_Float(A=$Particles.SpriteInitRotationRate, B=&VectorLength(Vector=$Particles.Velocity)))
   &Drag(Drag=10, Use Linear Drag=true, Use Rotational Drag=false)
   &SolveForcesAndVelocity(Clamp Velocity=false, Limit Acceleration=false)
   &ScaleColor(
    Scale Alpha=&FloatFromCurve(FloatCurve=DataInterface() {propertyValues {Curve {keys [{interpMode=RCIM_Cubic, tangentMode=RCTM_User, time=0.75, value=1}, <at 1 0.5>]}}})
    Scale RGB=&MakeVector(
     X=&UniformAOrBFloat(A=0)
     Y=1
     Z=&FloatFromCurve(FloatCurve=DataInterface() {propertyValues {Curve {keys [
          {arriveTangent=-172.0061, interpMode=RCIM_Cubic, leaveTangent=-172.0061, tangentMode=RCTM_User, time=0.2, value=10}
          {arriveTangent=-0.0507, interpMode=RCIM_Cubic, leaveTangent=-0.0508, tangentMode=RCTM_User, time=0.32, value=1}
          {arriveTangent=90.7103, interpMode=RCIM_Cubic, leaveTangent=90.7105, tangentMode=RCTM_User, time=1, value=20}
         ]}}})
    )
    ScaleA=true
    ScaleRGB=true
   ) {module=ScaleColor001}
   &ScaleSpriteSize(Scale Factor=&Vector2DFromFloat(Value=&FloatFromCurve(FloatCurve=DataInterface() {propertyValues {Curve {keys [<at 0 1 RCIM_Cubic>, <at 0.1 1.5 RCIM_Cubic>, {interpMode=RCIM_Cubic, leaveTangent=0.8761, time=0.5, value=1}]}}})))
  ]
  rendererClasses [&NiagaraSpriteRendererProperties]
  renderers [{rendererClass=&NiagaraSpriteRendererProperties}]
 }
 Squares {
  emitterUpdateScript [
   &EmitterState()
   &SpawnBurst_Instantaneous(Spawn Count=10, Use Spawn Probability=false)
  ]
  particleSpawnScript [
   /Game/Effects/Niagara/Modules/Orbit/FN_Orbit_Init_Variables_v2(Seed Offset:NiagaraInt32=&UniformRangedInt(Minimum=0, Maximum=1000))
   &ApplyInitialForces()
   &InitializeParticle(
    Lifetime=&UniformRangedFloat(Minimum=0.9, Maximum=1.5)
    Mesh Scale=&VectorFromFloat(Value=&UniformRangedFloat(Minimum=0.1, Maximum=0.6))
    Module.Write Color=true
    Module.Write Lifetime=true
    Module.Write Mass=true
    Module.Write Material Random=true
    Module.Write Position=true
    Module.Write Scale=true
    Module.Write SpriteRotation=false
    Module.Write SpriteSize=true
    Position=&SimulationPosition()
    Sprite Size=<5 5>
   )
   /Niagara/Modules/Spawn/Orientation/V2/InitialMeshOrientation(
    Orientation Axis:Vector3f=&UniformRangedVector(Minimum=<0 0 0>, Maximum=<1 1 1>)
    Orientation Vector=&UniformRangedVector(Minimum=<0 0 0>, Maximum=<1 1 1>)
    Use Orientation Vector=true
    Use Rotation Vector=false
   )
   /SetVariables(Particles.RotationDirection:NiagaraFloat=&UniformAOrBFloat(A=-1)) {isSetParameters=true}
   /SetVariables(
    Particles.ChooseVisibleMesh:NiagaraInt32=&UniformRangedInt(Minimum=0, Maximum=5)
    Particles.PickBone:NiagaraInt32=&UniformRangedInt(Minimum=0, Maximum=20)
    Particles.RotationAxis:Vector3f=&UniformRangedVector(Minimum=<0 0 0.5>, Maximum=<0.5 0.5 1>)
    Particles.RotationalVelocity:Vector3f=&UniformRangedVector(
     Minimum=&Multiply_VectorByFloat(Vector=&VectorFromFloat(Value=0.25), Float=$Particles.RotationDirection)
     Maximum=&Multiply_VectorByFloat(Vector=&VectorFromFloat(Value=0.5), Float=$Particles.RotationDirection)
    )
   ) {isSetParameters=true}
   /SetVariables(
    Particles.Tag1Visible:NiagaraInt32=&MakeCustomIntFromBool(
     Bool=&SetBoolByIntComparison(A=$Particles.ChooseVisibleMesh, B=1, Comparison Type=::NewEnumerator4)
     True INT=0
     False INT=1
    )
    Particles.Tag2Visible:NiagaraInt32=&MakeCustomIntFromBool(
     Bool=&SetBoolByIntComparison(A=$Particles.ChooseVisibleMesh, B=2, Comparison Type=::NewEnumerator4)
     True INT=0
     False INT=1
    )
    Particles.Tag3Visible:NiagaraInt32=&MakeCustomIntFromBool(
     Bool=&SetBoolByIntComparison(A=$Particles.ChooseVisibleMesh, B=3, Comparison Type=::NewEnumerator4)
     True INT=0
     False INT=1
    )
    Particles.Tag4Visible:NiagaraInt32=&MakeCustomIntFromBool(
     Bool=&SetBoolByIntComparison(A=$Particles.ChooseVisibleMesh, B=4, Comparison Type=::NewEnumerator4)
     True INT=0
     False INT=1
    )
    Particles.Tag5Visible:NiagaraInt32=&MakeCustomIntFromBool(
     Bool=&SetBoolByIntComparison(A=$Particles.ChooseVisibleMesh, B=5, Comparison Type=::NewEnumerator4)
     True INT=0
     False INT=1
    )
    Particles.Tag6Visible:NiagaraInt32=&MakeCustomIntFromBool(
     Bool=&SetBoolByIntComparison(A=$Particles.ChooseVisibleMesh, B=5, Comparison Type=::NewEnumerator4)
     True INT=0
     False INT=1
    )
   ) {isSetParameters=true}
  ]
  particleUpdateScript [
   &ParticleState()
   &SkeletalMeshLocation(
    Apply To Particle Position=false
    Bone / Socket Index=$Particles.PickBone
    Bone Sampling Mode=::NewEnumerator6
    Inherit Velocity=false
    MaskSpawnGroup=false
    Skeletal Mesh=$User.SkeletalMesh
    Use Sampled Mesh Attribute as Ribbon ID=false
   )
   &SolveForcesAndVelocity(Clamp Velocity=false, Limit Acceleration=false)
   &SolveRotationalForcesAndVelocity()
   /Game/Effects/Niagara/Modules/Orbit/FN_Orbit_v4(
    Height min/max:Vector2f=<-10 10>
    Radius min/max:Vector2f=&Lerp_Vector2(
     A=<0 0>
     B=<3 3>
     Alpha=&FloatFromCurve(FloatCurve=DataInterface() {propertyValues {Curve {keys [{arriveTangent=3.2407, interpMode=RCIM_Cubic, leaveTangent=3.2407, tangentMode=RCTM_User, time=0, value=0}, <at 1 1>]}}})
    )
    Rotation Offset min/max:Vector2f=&Multiply_Vector2(A=<0 360>, B=&Vector2DFromFloat(Value=$Particles.RotationDirection))
    Rotation Rate min/max:Vector2f=&Multiply_Vector2(
     A=&Lerp_Vector2(
      A=<0.05 0.075>
      B=<0.1 0.25>
      Alpha=&FloatFromCurve(FloatCurve=DataInterface() {propertyValues {Curve {keys [{arriveTangent=-2.4655, interpMode=RCIM_Cubic, leaveTangent=-2.4655, tangentMode=RCTM_User, time=0, value=1}, <at 1 0>]}}})
     )
     B=&Vector2DFromFloat(Value=$Particles.RotationDirection)
    )
    To Axis:Vector3f=$Particles.RotationAxis
   )
   /SetVariables(Particles.Timer:NiagaraFloat=&ClampFloat(
     Float=&Add_Float(
      A=&Multiply_Float(A=$Particles.NormalizedAge, B=10)
      B=&Multiply_Float(A=-0.5, B=$Particles.Initial.Lifetime)
     )
     Min=-0.1
     Max=1.1
    )) {isSetParameters=true}
   &ScaleColor(Scale RGB=&MakeVectorFromLinearColorRGB(Color=<rgba 1 1 1 1>), ScaleA=true, ScaleRGB=true)
   &ScaleMeshSize(Scale Factor=&VectorFromFloat(Value=&FloatFromCurve(FloatCurve=DataInterface() {propertyValues {Curve {keys [
          {interpMode=RCIM_Cubic, tangentMode=RCTM_User, time=0, value=0}
          {arriveTangent=3.3333, interpMode=RCIM_Cubic, leaveTangent=3.3333, tangentMode=RCTM_User, time=0.125, value=1.2}
          {arriveTangent=-1.6969, interpMode=RCIM_Cubic, leaveTangent=-1.6969, tangentMode=RCTM_User, time=0.175, value=1}
          {arriveTangent=1.4531, interpMode=RCIM_Cubic, leaveTangent=1.4531, tangentMode=RCTM_User, time=0.8933, value=1}
          {arriveTangent=-1.4286, interpMode=RCIM_Cubic, leaveTangent=-1.4286, tangentMode=RCTM_User, time=0.9433, value=1.2}
          {interpMode=RCIM_Cubic, tangentMode=RCTM_User, time=1, value=0}
         ]}}})))
   /SetVariables(Particles.MeshOrientation=&LerpQuaternion(
     Quaternion A:Quat4f=$Particles.MeshOrientation
     Quaternion B:Quat4f=$Particles.FN_Orbit_v4.MeshOrientation
     Lerp Factor:NiagaraFloat=$Particles.NormalizedAge
     Enable Spherical Linear Interpolation:NiagaraBool=false
    )) {isSetParameters=true}
   &self:TransformToBoneSpace(
    BonePosition:Vector3f=$Output.SkeletalMeshLocation.SampledPosition
    BoneRotation:Quat4f=$Output.SkeletalMeshLocation.SampledRotation
    Position:Vector3f=$Particles.FN_Orbit_v4.Position
   ) {module=TransformToBoneSpace}
  ]
  rendererClasses [&NiagaraSpriteRendererProperties, &NiagaraMeshRendererProperties]
  renderers [
   {rendererClass=&NiagaraSpriteRendererProperties}
   {rendererClass=&NiagaraMeshRendererProperties, rendererIndex=1}
   {rendererClass=&NiagaraMeshRendererProperties, rendererIndex=2}
   {rendererClass=&NiagaraMeshRendererProperties, rendererIndex=3}
   {rendererClass=&NiagaraMeshRendererProperties, rendererIndex=4}
   {rendererClass=&NiagaraMeshRendererProperties, rendererIndex=5}
   {rendererClass=&NiagaraMeshRendererProperties, rendererIndex=6}
  ]
 }
}
summary {
 systemName=NS_Pickaxe_CloakedAssassin_Burst
 emitters [
  {emitterName=LightningEndPicker, rendererClasses [&NiagaraSpriteRendererProperties]}
  {
   emitterName=LightningRibbon
   rendererClasses [&NiagaraRibbonRendererProperties, &NiagaraSpriteRendererProperties]
  }
  {emitterName=LightningSprites, rendererClasses [&NiagaraSpriteRendererProperties]}
  {emitterName=BigLightningEndPicker, rendererClasses [&NiagaraSpriteRendererProperties]}
  {
   emitterName=BigLightningRibbon
   rendererClasses [&NiagaraRibbonRendererProperties, &NiagaraSpriteRendererProperties]
  }
  {emitterName=Squares, rendererClasses [&NiagaraSpriteRendererProperties, &NiagaraMeshRendererProperties]}
  {emitterName=Decal, rendererClasses [&NiagaraSpriteRendererProperties]}
  {emitterName=Runes, rendererClasses [&NiagaraSpriteRendererProperties]}
 ]
 userVariables [
  User.bIsFrontEndPreview {type=NiagaraBool}
  User.DamageLevel {type=NiagaraInt32}
  User.SkeletalMesh {
   type=NiagaraDataInterfaceSkeletalMesh
   defaultValue {
    struct=NiagaraExt_VariableValue_DataInterface
    value {dataInterface=&self:NiagaraDataInterfaceSkeletalMesh_1}
   }
  }
 ]
}
