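# Translucent, two-sided glass material for the female Cloaked Assassin pickaxe, with black base color and screen-space reflections.
# Emissive blends the outer and inner color parameters using an inverted, contrast-adjusted sphere mask along the U coordinate, faking a gradient from UVs alone.
# An inverted radial UV mask scaled by Glass Opacity drives opacity, fading the center.
# The attributes pass through the MF_TechArtOverrides_accessories function before output.
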Material /BRCosmetics/Weapons/FORT_Melee/Pickaxe_Cloaked_Assassin_Female/Materials/DeferedGlass_CloakedAssassin
graph {
 properties {
  bScreenSpaceReflections=true
  bUseMaterialAttributes=true
  bUsedWithMeshDeformer=true
  bUsedWithNanite=true
  bUsedWithSkeletalMesh=true
  blendMode=BLEND_Translucent
  pixelDepthOffsetMode=PDOM_Legacy
  translucencyLightingMode=TLM_VolumetricDirectional
  twoSided=true
  usedShadingModels=Default Lit
 }
 nodes [
  n1 = VectorParameter() {parameterName=Glass Outer Color, defaultValue=<rgba 0.6219 0.099 1 1>}
  n2 = VectorParameter() {parameterName=Glass Inner Color, defaultValue=<rgba 0.0752 0.0449 0.1198 1>}
  n3 = MaterialFunctionCall(Float2=TextureCoordinate()) {materialFunction=&BreakOutFloat2Components}
  n4 = Constant() {r=0.5}
  n5 = Constant() {r=0.6}
  n6 = Constant() {r=0.3}
  n7 = MaterialFunctionCall(In=OneMinus(None=SphereMask(A=@n3.R, B=@n4, Radius=@n5, Hardness=@n6)), Contrast=Constant() {r=1}) {materialFunction=&CheapContrast}
  n8 = OneMinus(None=SphereMask(A=TextureCoordinate(), B=Constant2Vector() {g=0.5, r=0.5}, Radius=Constant() {r=0.55}, Hardness=Constant() {r=0.5}))
  n9 = ScalarParameter() {defaultValue=0.12, parameterName=Glass Opacity}
  MaterialFunctionCall(In=SetMaterialAttributes!(
    MaterialAttributes
    Emissive Color=LinearInterpolate(A=@n1.RGB, B=@n2.RGB, Alpha=@n7.Result)
    Opacity=LinearInterpolate(B=@n8, Alpha=@n9)
    Base Color=Constant()
   ) {attributeSetTypes [#1, #2, #3]}) {materialFunction=&MF_TechArtOverrides_accessories}
 ]
}
