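# Animation Blueprint for the Cloaked Assassin Female pickaxe, animating its floating cube and base bones.
# Poses copied from the attached parent mesh get additive bone offsets from float vector and rotator variables, plus zero-gravity, spring-constrained Anim Dynamics.
# Nested bool blends select separately tuned chains for front end, melee swings, and on or off ground.
# The melee chain skips the float offsets, and the front-end chain also nudges the energy plane bone.
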AnimBlueprint /BRCosmetics/Weapons/FORT_Melee/Pickaxe_Cloaked_Assassin_Female/Meshes/Cloaked_Assassin_Female_AnimBP
target_skeleton=&/Cloaked_Assassin_Female_Skeleton
graphs [
 AnimGraph {nodes [
   n1 = Get Cube_Float_A() -> (Cube_Float_A:Vector)
   n2 = Get Floating_Rotator2() -> (Floating_Rotator2:Rotator)
   n3 = "Transform (Modify) Bone - Bone: DYN_Cube1_2"(
    ComponentPose=Local To Component(LocalPose=Copy Pose From Mesh() {bCopyCurves=true, bUseAttachedParent=true})
    Alpha=1
    Translation=@n1
    Rotation=@n2
    Scale="1,1,1"
   ) {rotationMode=BMM_Additive, translationMode=BMM_Additive, alphaBoolBlend {blendOption=Linear}, boneToModify {boneName=DYN_Cube1_2}}
   n4 = Get Cube_Float_C() -> (Cube_Float_C:Vector)
   n5 = Get Floating_Rotator1() -> (Floating_Rotator1:Rotator)
   n6 = "Transform (Modify) Bone - Bone: DYN_Cube2_2"(ComponentPose=@n3, Alpha=1, Translation=@n4, Rotation=@n5, Scale="1,1,1") {rotationMode=BMM_Additive, translationMode=BMM_Additive, alphaBoolBlend {blendOption=Linear}, boneToModify {boneName=DYN_Cube2_2}}
   n7 = "Transform (Modify) Bone - Bone: DYN_Cube3_2"(ComponentPose=@n6, Alpha=0.5, Translation=Get Cube_Float_B() -> (Cube_Float_B:Vector), Rotation=@n2, Scale="1,1,1") {rotationMode=BMM_Additive, translationMode=BMM_Additive, alphaBoolBlend {blendOption=Linear}, boneToModify {boneName=DYN_Cube3_2}}
   n8 = "Transform (Modify) Bone - Bone: DYN_Cube4_2"(ComponentPose=@n7, Alpha=1, Translation=Get Cube_Float_D() -> (Cube_Float_D:Vector), Rotation=@n5, Scale="1,1,1") {rotationMode=BMM_Additive, translationMode=BMM_Additive, alphaBoolBlend {blendOption=Linear}, boneToModify {boneName=DYN_Cube4_2}}
   n9 = "Transform (Modify) Bone - Bone: DYN_Cube5_2"(ComponentPose=@n8, Alpha=1, Translation=@n4, Rotation=@n5, Scale="1,1,1") {rotationMode=BMM_Additive, translationMode=BMM_Additive, alphaBoolBlend {blendOption=Linear}, boneToModify {boneName=DYN_Cube5_2}}
   n10 = "Transform (Modify) Bone - Bone: DYN_Cube6_2"(ComponentPose=@n9, Alpha=0.5, Translation=@n1, Rotation=@n2, Scale="1,1,1") {rotationMode=BMM_Additive, translationMode=BMM_Additive, alphaBoolBlend {blendOption=Linear}, boneToModify {boneName=DYN_Cube6_2}}
   n11 = "Transform (Modify) Bone - Bone: DYN_Cube1_4"(ComponentPose=@n10, Alpha=1, Translation=@n4, Rotation=@n5, Scale="1,1,1") {rotationMode=BMM_Additive, translationMode=BMM_Additive, alphaBoolBlend {blendOption=Linear}, boneToModify {boneName=DYN_Cube1_4}}
   n12 = "Transform (Modify) Bone - Bone: DYN_Base"(ComponentPose=@n11, Alpha=0.7, Translation=@n1, Rotation=@n5, Scale="1,1,1") {
    rotationMode=BMM_Additive
    translationMode=BMM_Additive
    translationSpace=BCS_ParentBoneSpace
    alphaBoolBlend {blendOption=Linear}
    boneToModify {boneName=DYN_Base}
   }
   n13 = Anim Dynamics - Bone: DYN_Cube1_1(ComponentPose=@n12, Alpha=0.7, ExternalForce="0,0,0") {
    angularDampingOverride=0.9
    angularSpringConstant=150
    bAngularSpring=true
    bGravityOverrideInSimSpace=true
    bLinearSpring=true
    bOverrideAngularDamping=true
    bOverrideLinearDamping=true
    gravityScale=0
    linearDampingOverride=0.995
    linearSpringConstant=30
    simulationSpace=RootRelative
    alphaBoolBlend {blendOption=Linear}
    boundBone {boneName=DYN_Cube1_1}
    componentLinearAccScale=<0.06 0.06 0.06>
    componentLinearVelScale=<0.8 0.8 0.6>
    physicsBodyDefinitions [{
      sphereCollisionRadius=0
      boundBone {boneName=DYN_Cube1_1}
      boxExtents=<50 50 50>
      constraintSetup {angularLimitsMax=<0 0 0>, angularLimitsMin=<0 0 0>, angularTarget=<0 0 0>, linearAxesMax=<2 5 8>, linearAxesMin=<-2 -1 -8>}
      localJointOffset=<0 -3 0>
     }]
    relativeSpaceBone {boneName=b_AR_Root}
   }
   n14 = Anim Dynamics - Bone: DYN_Cube3_1(ComponentPose=@n13, Alpha=0.7, ExternalForce="0,0,0") {
    angularDampingOverride=0.9
    angularSpringConstant=150
    bAngularSpring=true
    bGravityOverrideInSimSpace=true
    bLinearSpring=true
    bOverrideAngularDamping=true
    bOverrideLinearDamping=true
    gravityScale=0
    linearDampingOverride=0.995
    linearSpringConstant=30
    simulationSpace=RootRelative
    alphaBoolBlend {blendOption=Linear}
    boundBone {boneName=DYN_Cube3_1}
    componentLinearAccScale=<0.1 0.1 0.1>
    componentLinearVelScale=<0.8 0.8 0.6>
    physicsBodyDefinitions [{
      sphereCollisionRadius=0
      boundBone {boneName=DYN_Cube3_1}
      boxExtents=<50 50 50>
      constraintSetup {angularLimitsMax=<0 0 0>, angularLimitsMin=<0 0 0>, angularTarget=<0 0 0>, linearAxesMax=<2 5 8>, linearAxesMin=<-2 0 -8>}
      localJointOffset=<0 -3 0>
     }]
    relativeSpaceBone {boneName=b_AR_Root}
   }
   n15 = Anim Dynamics - Bone: DYN_Cube5_1(ComponentPose=@n14, Alpha=0.7, ExternalForce="0,0,0") {
    angularDampingOverride=0.9
    angularSpringConstant=150
    bAngularSpring=true
    bGravityOverrideInSimSpace=true
    bLinearSpring=true
    bOverrideAngularDamping=true
    bOverrideLinearDamping=true
    gravityScale=0
    linearDampingOverride=0.995
    linearSpringConstant=30
    simulationSpace=RootRelative
    alphaBoolBlend {blendOption=Linear}
    boundBone {boneName=DYN_Cube5_1}
    componentLinearAccScale=<0.05 0.05 0.05>
    componentLinearVelScale=<0.8 1 0.6>
    physicsBodyDefinitions [{
      sphereCollisionRadius=0
      boundBone {boneName=DYN_Cube5_1}
      boxExtents=<50 50 50>
      constraintSetup {angularLimitsMax=<0 0 0>, angularLimitsMin=<0 0 0>, angularTarget=<0 0 0>, linearAxesMax=<2 5 8>, linearAxesMin=<-2 1 -8>}
      localJointOffset=<0 -3 0>
     }]
    relativeSpaceBone {boneName=b_AR_Root}
   }
   n16 = Anim Dynamics - Bone: DYN_Cube2_1(ComponentPose=@n15, Alpha=0.7, ExternalForce="0,0,0") {
    angularDampingOverride=0.9
    angularSpringConstant=150
    bAngularSpring=true
    bGravityOverrideInSimSpace=true
    bLinearSpring=true
    bOverrideAngularDamping=true
    bOverrideLinearDamping=true
    gravityScale=0
    linearDampingOverride=0.995
    linearSpringConstant=30
    simulationSpace=RootRelative
    alphaBoolBlend {blendOption=Linear}
    boundBone {boneName=DYN_Cube2_1}
    componentLinearAccScale=<0.07 0.07 0.07>
    componentLinearVelScale=<0.8 0.8 0.6>
    physicsBodyDefinitions [{
      sphereCollisionRadius=0
      boundBone {boneName=DYN_Cube2_1}
      boxExtents=<50 50 50>
      constraintSetup {angularLimitsMax=<0 0 0>, angularLimitsMin=<0 0 0>, angularTarget=<0 0 0>, linearAxesMax=<2 5 8>, linearAxesMin=<-2 -1 -8>}
      localJointOffset=<0 -3 0>
     }]
    relativeSpaceBone {boneName=b_AR_Root}
   }
   n17 = Anim Dynamics - Bone: DYN_Cube4_1(ComponentPose=@n16, Alpha=0.7, ExternalForce="0,0,0") {
    angularDampingOverride=0.9
    angularSpringConstant=150
    bAngularSpring=true
    bGravityOverrideInSimSpace=true
    bLinearSpring=true
    bOverrideAngularDamping=true
    bOverrideLinearDamping=true
    gravityScale=0
    linearDampingOverride=0.995
    linearSpringConstant=30
    simulationSpace=RootRelative
    alphaBoolBlend {blendOption=Linear}
    boundBone {boneName=DYN_Cube4_1}
    componentLinearAccScale=<0.09 0.09 0.09>
    componentLinearVelScale=<0.8 0.8 0.6>
    physicsBodyDefinitions [{
      sphereCollisionRadius=0
      boundBone {boneName=DYN_Cube4_1}
      boxExtents=<50 50 50>
      constraintSetup {angularLimitsMax=<0 0 0>, angularLimitsMin=<0 0 0>, angularTarget=<0 0 0>, linearAxesMax=<2 5 8>, linearAxesMin=<-2 0 -8>}
      localJointOffset=<0 -3 0>
     }]
    relativeSpaceBone {boneName=b_AR_Root}
   }
   n18 = Anim Dynamics - Bone: DYN_Cube6_1(ComponentPose=@n17, Alpha=0.7, ExternalForce="0,0,0") {
    angularDampingOverride=0.9
    angularSpringConstant=150
    bAngularSpring=true
    bGravityOverrideInSimSpace=true
    bLinearSpring=true
    bOverrideAngularDamping=true
    bOverrideLinearDamping=true
    gravityScale=0
    linearDampingOverride=0.995
    linearSpringConstant=30
    simulationSpace=RootRelative
    alphaBoolBlend {blendOption=Linear}
    boundBone {boneName=DYN_Cube6_1}
    componentLinearAccScale=<0.1 0.1 0.1>
    componentLinearVelScale=<0.8 1 0.6>
    physicsBodyDefinitions [{
      sphereCollisionRadius=0
      boundBone {boneName=DYN_Cube6_1}
      boxExtents=<50 50 50>
      constraintSetup {angularLimitsMax=<0 0 0>, angularLimitsMin=<0 0 0>, angularTarget=<0 0 0>, linearAxesMax=<2 5 8>, linearAxesMin=<-2 1 -8>}
      localJointOffset=<0 -3 0>
     }]
    relativeSpaceBone {boneName=b_AR_Root}
   }
   n19 = Anim Dynamics - Bone: DYN_Cube1_3(ComponentPose=@n18, Alpha=0.7, ExternalForce="0,0,0") {
    angularDampingOverride=0.9
    angularSpringConstant=150
    bAngularSpring=true
    bGravityOverrideInSimSpace=true
    bLinearSpring=true
    bOverrideAngularDamping=true
    bOverrideLinearDamping=true
    gravityScale=0
    linearDampingOverride=0.995
    linearSpringConstant=30
    simulationSpace=RootRelative
    alphaBoolBlend {blendOption=Linear}
    boundBone {boneName=DYN_Cube1_3}
    componentLinearAccScale=<0.07 0.07 0.07>
    componentLinearVelScale=<0.8 0.8 0.6>
    physicsBodyDefinitions [{
      sphereCollisionRadius=0
      boundBone {boneName=DYN_Cube1_3}
      boxExtents=<50 50 50>
      constraintSetup {angularLimitsMax=<0 0 0>, angularLimitsMin=<0 0 0>, angularTarget=<0 0 0>, linearAxesMax=<2 5 8>, linearAxesMin=<-2 -1 -8>}
      localJointOffset=<0 -3 0>
     }]
    relativeSpaceBone {boneName=b_AR_Root}
   }
   n20 = Anim Dynamics - Bone: DYN_Base(ComponentPose=@n19, Alpha=0.5, ExternalForce="0,0,0") {
    angularDampingOverride=0.9
    angularSpringConstant=150
    bAngularSpring=true
    bGravityOverrideInSimSpace=true
    bLinearSpring=true
    bOverrideAngularDamping=true
    bOverrideLinearDamping=true
    gravityScale=0
    linearDampingOverride=0.995
    linearSpringConstant=40
    simulationSpace=BoneRelative
    alphaBoolBlend {blendOption=Linear}
    boundBone {boneName=DYN_Base}
    componentLinearVelScale=<0.8 0.8 0.8>
    gravityOverride {z=-980}
    physicsBodyDefinitions [{
      sphereCollisionRadius=0
      boundBone {boneName=DYN_Base}
      boxExtents=<50 50 50>
      constraintSetup {angularLimitsMax=<0 0 0>, angularLimitsMin=<0 0 0>, angularTarget=<0 0 0>, linearAxesMax=<4 4 4>, linearAxesMin=<-4 -4 -4>}
      localJointOffset=<3 0 0>
     }]
    relativeSpaceBone {boneName=b_AR_Root}
   }
   n21 = Get Cube_Float_A()
   n22 = Get Floating_Rotator2()
   n23 = "Transform (Modify) Bone - Bone: DYN_Cube1_2"(ComponentPose=Local To Component(LocalPose=Use cached pose 'OnGroundCache'()), Alpha=1, Translation=@n21, Rotation=@n22, Scale="1,1,1") {rotationMode=BMM_Additive, translationMode=BMM_Additive, alphaBoolBlend {blendOption=Linear}, boneToModify {boneName=DYN_Cube1_2}}
   n24 = Get Floating_Rotator1()
   n25 = Get Cube_Float_C()
   n26 = "Transform (Modify) Bone - Bone: DYN_Cube2_2"(ComponentPose=@n23, Alpha=1, Translation=@n25, Rotation=@n24, Scale="1,1,1") {rotationMode=BMM_Additive, translationMode=BMM_Additive, alphaBoolBlend {blendOption=Linear}, boneToModify {boneName=DYN_Cube2_2}}
   n27 = "Transform (Modify) Bone - Bone: DYN_Cube3_2"(ComponentPose=@n26, Alpha=0.5, Translation=Get Cube_Float_B(), Rotation=@n22, Scale="1,1,1") {rotationMode=BMM_Additive, translationMode=BMM_Additive, alphaBoolBlend {blendOption=Linear}, boneToModify {boneName=DYN_Cube3_2}}
   n28 = "Transform (Modify) Bone - Bone: DYN_Cube4_2"(ComponentPose=@n27, Alpha=1, Translation=Get Cube_Float_D(), Rotation=@n24, Scale="1,1,1") {rotationMode=BMM_Additive, translationMode=BMM_Additive, alphaBoolBlend {blendOption=Linear}, boneToModify {boneName=DYN_Cube4_2}}
   n29 = "Transform (Modify) Bone - Bone: DYN_Cube5_2"(ComponentPose=@n28, Alpha=1, Translation=@n25, Rotation=@n24, Scale="1,1,1") {rotationMode=BMM_Additive, translationMode=BMM_Additive, alphaBoolBlend {blendOption=Linear}, boneToModify {boneName=DYN_Cube5_2}}
   n30 = "Transform (Modify) Bone - Bone: DYN_Cube6_2"(ComponentPose=@n29, Alpha=0.5, Translation=@n21, Rotation=@n22, Scale="1,1,1") {rotationMode=BMM_Additive, translationMode=BMM_Additive, alphaBoolBlend {blendOption=Linear}, boneToModify {boneName=DYN_Cube6_2}}
   n31 = "Transform (Modify) Bone - Bone: DYN_Cube2_2"(ComponentPose=@n30, Alpha=1, Translation=@n25, Rotation=@n24, Scale="1,1,1") {rotationMode=BMM_Additive, translationMode=BMM_Additive, alphaBoolBlend {blendOption=Linear}, boneToModify {boneName=DYN_Cube2_2}}
   n32 = "Transform (Modify) Bone - Bone: DYN_Base"(ComponentPose=@n31, Alpha=0.7, Translation=@n21, Rotation=@n22, Scale="1,1,1") {
    rotationMode=BMM_Additive
    translationMode=BMM_Additive
    translationSpace=BCS_BoneSpace
    alphaBoolBlend {blendOption=Linear}
    boneToModify {boneName=DYN_Base}
   }
   n33 = Anim Dynamics - Bone: DYN_Cube1_1(ComponentPose=@n32, Alpha=0.7, ExternalForce="0,0,0") {
    angularDampingOverride=0.9
    angularSpringConstant=210
    bAngularSpring=true
    bGravityOverrideInSimSpace=true
    bLinearSpring=true
    bOverrideAngularDamping=true
    bOverrideLinearDamping=true
    gravityScale=0
    linearDampingOverride=0.9999
    linearSpringConstant=120
    simulationSpace=BoneRelative
    alphaBoolBlend {blendOption=Linear}
    boundBone {boneName=DYN_Cube1_1}
    componentAppliedLinearAccClamp=<6000 6000 6000>
    componentLinearAccScale=<0.05 0.05 0.05>
    componentLinearVelScale=<0.9 0.9 0.9>
    gravityOverride {z=-980}
    physicsBodyDefinitions [{
      sphereCollisionRadius=0
      boundBone {boneName=DYN_Cube1_1}
      boxExtents=<50 50 50>
      constraintSetup {angularLimitsMax=<0 0 0>, angularLimitsMin=<0 0 0>, angularTarget=<0 0 0>, linearAxesMax=<3 5 5>, linearAxesMin=<-3 -1 -5>}
      localJointOffset=<-3 0 0>
     }]
    relativeSpaceBone {boneName=b_AR_Root}
   }
   n34 = Anim Dynamics - Bone: DYN_Cube3_1(ComponentPose=@n33, Alpha=0.5, ExternalForce="0,0,0") {
    angularDampingOverride=0.9
    angularSpringConstant=210
    bAngularSpring=true
    bGravityOverrideInSimSpace=true
    bLinearSpring=true
    bOverrideAngularDamping=true
    bOverrideLinearDamping=true
    gravityScale=0
    linearDampingOverride=0.9999
    linearSpringConstant=210
    simulationSpace=BoneRelative
    alphaBoolBlend {blendOption=Linear}
    boundBone {boneName=DYN_Cube3_1}
    componentAppliedLinearAccClamp=<6000 6000 6000>
    componentLinearVelScale=<0.9 0.9 0.9>
    gravityOverride {z=-980}
    physicsBodyDefinitions [{
      sphereCollisionRadius=0
      boundBone {boneName=DYN_Cube3_1}
      boxExtents=<50 50 50>
      constraintSetup {angularLimitsMax=<0 0 0>, angularLimitsMin=<0 0 0>, angularTarget=<0 0 0>, linearAxesMax=<2 5 4>, linearAxesMin=<-2 0 -4>}
      localJointOffset=<-3 0 0>
     }]
    relativeSpaceBone {boneName=b_AR_Root}
   }
   n35 = Anim Dynamics - Bone: DYN_Cube5_1(ComponentPose=@n34, Alpha=0.3, ExternalForce="0,0,0") {
    angularDampingOverride=0.9
    angularSpringConstant=210
    bAngularSpring=true
    bGravityOverrideInSimSpace=true
    bLinearSpring=true
    bOverrideAngularDamping=true
    bOverrideLinearDamping=true
    gravityScale=0
    linearDampingOverride=0.9999
    linearSpringConstant=210
    simulationSpace=BoneRelative
    alphaBoolBlend {blendOption=Linear}
    boundBone {boneName=DYN_Cube5_1}
    componentAppliedLinearAccClamp=<6000 6000 6000>
    componentLinearAccScale=<0.05 0.05 0.05>
    componentLinearVelScale=<0.9 0.9 0.9>
    gravityOverride {z=-980}
    physicsBodyDefinitions [{
      sphereCollisionRadius=0
      boundBone {boneName=DYN_Cube5_1}
      boxExtents=<50 50 50>
      constraintSetup {angularLimitsMax=<0 0 0>, angularLimitsMin=<0 0 0>, angularTarget=<0 0 0>, linearAxesMax=<2 5 4>, linearAxesMin=<-2 1 -4>}
      localJointOffset=<-3 0 0>
     }]
    relativeSpaceBone {boneName=b_AR_Root}
   }
   n36 = Anim Dynamics - Bone: DYN_Cube2_1(ComponentPose=@n35, Alpha=0.7, ExternalForce="0,0,0") {
    angularDampingOverride=0.9
    angularSpringConstant=210
    bAngularSpring=true
    bGravityOverrideInSimSpace=true
    bLinearSpring=true
    bOverrideAngularDamping=true
    bOverrideLinearDamping=true
    gravityScale=0
    linearDampingOverride=0.9999
    linearSpringConstant=210
    simulationSpace=BoneRelative
    alphaBoolBlend {blendOption=Linear}
    boundBone {boneName=DYN_Cube2_1}
    componentAppliedLinearAccClamp=<6000 6000 6000>
    componentLinearVelScale=<0.9 0.9 0.9>
    gravityOverride {z=-980}
    physicsBodyDefinitions [{
      sphereCollisionRadius=0
      boundBone {boneName=DYN_Cube2_1}
      boxExtents=<50 50 50>
      constraintSetup {angularLimitsMax=<0 0 0>, angularLimitsMin=<0 0 0>, angularTarget=<0 0 0>, linearAxesMax=<2 5 4>, linearAxesMin=<-2 -1 -4>}
      localJointOffset=<-3 0 0>
     }]
    relativeSpaceBone {boneName=b_AR_Root}
   }
   n37 = Anim Dynamics - Bone: DYN_Cube4_1(ComponentPose=@n36, Alpha=0.6, ExternalForce="0,0,0") {
    angularDampingOverride=0.9
    angularSpringConstant=210
    bAngularSpring=true
    bGravityOverrideInSimSpace=true
    bLinearSpring=true
    bOverrideAngularDamping=true
    bOverrideLinearDamping=true
    gravityScale=0
    linearDampingOverride=0.9999
    linearSpringConstant=165
    simulationSpace=BoneRelative
    alphaBoolBlend {blendOption=Linear}
    boundBone {boneName=DYN_Cube4_1}
    componentAppliedLinearAccClamp=<6000 6000 6000>
    componentLinearAccScale=<0.05 0.05 0.05>
    componentLinearVelScale=<0.9 0.9 0.9>
    gravityOverride {z=-980}
    physicsBodyDefinitions [{
      sphereCollisionRadius=0
      boundBone {boneName=DYN_Cube4_1}
      boxExtents=<50 50 50>
      constraintSetup {angularLimitsMax=<0 0 0>, angularLimitsMin=<0 0 0>, angularTarget=<0 0 0>, linearAxesMax=<2 5 4>, linearAxesMin=<-2 0 -4>}
      localJointOffset=<-3 0 0>
     }]
    relativeSpaceBone {boneName=b_AR_Root}
   }
   n38 = Anim Dynamics - Bone: DYN_Cube6_1(ComponentPose=@n37, Alpha=1, ExternalForce="0,0,0") {
    angularDampingOverride=0.9
    angularSpringConstant=210
    bAngularSpring=true
    bGravityOverrideInSimSpace=true
    bLinearSpring=true
    bOverrideAngularDamping=true
    bOverrideLinearDamping=true
    gravityScale=0
    linearDampingOverride=0.9999
    linearSpringConstant=195
    simulationSpace=BoneRelative
    alphaBoolBlend {blendOption=Linear}
    boundBone {boneName=DYN_Cube6_1}
    componentAppliedLinearAccClamp=<6000 6000 6000>
    componentLinearAccScale=<0.1 0.1 0.1>
    componentLinearVelScale=<0.9 0.9 0.9>
    gravityOverride {z=-980}
    physicsBodyDefinitions [{
      sphereCollisionRadius=0
      boundBone {boneName=DYN_Cube6_1}
      boxExtents=<50 50 50>
      constraintSetup {angularLimitsMax=<0 0 0>, angularLimitsMin=<0 0 0>, angularTarget=<0 0 0>, linearAxesMax=<4 5 5>, linearAxesMin=<-4 0 -5>}
      localJointOffset=<-3 0 0>
     }]
    relativeSpaceBone {boneName=b_AR_Root}
   }
   n39 = Anim Dynamics - Bone: DYN_Cube1_3(ComponentPose=@n38, Alpha=0.7, ExternalForce="0,0,0") {
    angularDampingOverride=0.9
    angularSpringConstant=210
    bAngularSpring=true
    bGravityOverrideInSimSpace=true
    bLinearSpring=true
    bOverrideAngularDamping=true
    bOverrideLinearDamping=true
    gravityScale=0
    linearDampingOverride=0.9999
    linearSpringConstant=210
    simulationSpace=BoneRelative
    alphaBoolBlend {blendOption=Linear}
    boundBone {boneName=DYN_Cube1_3}
    componentAppliedLinearAccClamp=<6000 6000 6000>
    componentLinearVelScale=<0.9 0.9 0.9>
    gravityOverride {z=-980}
    physicsBodyDefinitions [{
      sphereCollisionRadius=0
      boundBone {boneName=DYN_Cube1_3}
      boxExtents=<50 50 50>
      constraintSetup {angularLimitsMax=<0 0 0>, angularLimitsMin=<0 0 0>, angularTarget=<0 0 0>, linearAxesMax=<2 5 4>, linearAxesMin=<-2 -1 -4>}
      localJointOffset=<-3 0 0>
     }]
    relativeSpaceBone {boneName=b_AR_Root}
   }
   n40 = Anim Dynamics - Bone: DYN_Base(ComponentPose=@n39, Alpha=0.3, ExternalForce="0,0,0") {
    angularDampingOverride=0.9
    angularSpringConstant=210
    bAngularSpring=true
    bGravityOverrideInSimSpace=true
    bLinearSpring=true
    bOverrideAngularDamping=true
    bOverrideLinearDamping=true
    gravityScale=0
    linearDampingOverride=0.9999
    linearSpringConstant=210
    simulationSpace=BoneRelative
    alphaBoolBlend {blendOption=Linear}
    boundBone {boneName=DYN_Base}
    componentAppliedLinearAccClamp=<6000 6000 6000>
    componentLinearVelScale=<0.9 0.9 0.9>
    gravityOverride {z=-980}
    physicsBodyDefinitions [{
      sphereCollisionRadius=0
      boundBone {boneName=DYN_Base}
      boxExtents=<50 50 50>
      constraintSetup {angularLimitsMax=<0 0 0>, angularLimitsMin=<0 0 0>, angularTarget=<0 0 0>, linearAxesMax=<2 5 4>, linearAxesMin=<-2 -1 -4>}
      localJointOffset=<3 0 0>
     }]
    relativeSpaceBone {boneName=b_AR_Root}
   }
   n41 = Blend Poses by bool(
    BlendPose_0=Component To Local(ComponentPose=@n20)
    BlendPose_1=Component To Local(ComponentPose=@n40)
    BlendTime_1=1
    bActiveValue=Get bIsOnGround() -> (bIsOnGround:bool)
   ) {blendType=Linear}
   n42 = Anim Dynamics - Bone: DYN_Cube1_1(
    ComponentPose=Local To Component(LocalPose=Copy Pose From Mesh() {bCopyCurves=true, bUseAttachedParent=true})
    Alpha=0.8
    ExternalForce="0,0,0"
   ) {
    angularDampingOverride=0.9
    angularSpringConstant=150
    bAngularSpring=true
    bGravityOverrideInSimSpace=true
    bLinearSpring=true
    bOverrideAngularDamping=true
    bOverrideLinearDamping=true
    gravityScale=0
    linearDampingOverride=0.8
    linearSpringConstant=30
    alphaBoolBlend {blendOption=Linear}
    boundBone {boneName=DYN_Cube1_1}
    componentLinearAccScale=<0.16 0.16 0.16>
    componentLinearVelScale=<0.8 0.8 0.6>
    physicsBodyDefinitions [{
      sphereCollisionRadius=0
      boundBone {boneName=DYN_Cube1_1}
      boxExtents=<50 50 50>
      constraintSetup {angularLimitsMax=<0 0 0>, angularLimitsMin=<0 0 0>, angularTarget=<0 0 0>, linearAxesMax=<2 5 8>, linearAxesMin=<-2 -1 -8>}
      localJointOffset=<0 -3 0>
     }]
    relativeSpaceBone {boneName=b_AR_Root}
   }
   n43 = Anim Dynamics - Bone: DYN_Cube3_1(ComponentPose=@n42, Alpha=0.8, ExternalForce="0,0,0") {
    angularDampingOverride=0.9
    angularSpringConstant=150
    bAngularSpring=true
    bGravityOverrideInSimSpace=true
    bLinearSpring=true
    bOverrideAngularDamping=true
    bOverrideLinearDamping=true
    gravityScale=0
    linearDampingOverride=0.8
    linearSpringConstant=30
    alphaBoolBlend {blendOption=Linear}
    boundBone {boneName=DYN_Cube3_1}
    componentLinearAccScale=<0.2 0.2 0.2>
    componentLinearVelScale=<0.8 0.8 0.6>
    physicsBodyDefinitions [{
      sphereCollisionRadius=0
      boundBone {boneName=DYN_Cube3_1}
      boxExtents=<50 50 50>
      constraintSetup {angularLimitsMax=<0 0 0>, angularLimitsMin=<0 0 0>, angularTarget=<0 0 0>, linearAxesMax=<2 5 8>, linearAxesMin=<-2 0 -8>}
      localJointOffset=<0 -3 0>
     }]
    relativeSpaceBone {boneName=b_AR_Root}
   }
   n44 = Anim Dynamics - Bone: DYN_Cube5_1(ComponentPose=@n43, Alpha=0.8, ExternalForce="0,0,0") {
    angularDampingOverride=0.9
    angularSpringConstant=150
    bAngularSpring=true
    bGravityOverrideInSimSpace=true
    bLinearSpring=true
    bOverrideAngularDamping=true
    bOverrideLinearDamping=true
    gravityScale=0
    linearDampingOverride=0.8
    linearSpringConstant=30
    alphaBoolBlend {blendOption=Linear}
    boundBone {boneName=DYN_Cube5_1}
    componentLinearAccScale=<0.15 0.15 0.15>
    componentLinearVelScale=<0.8 1 0.6>
    physicsBodyDefinitions [{
      sphereCollisionRadius=0
      boundBone {boneName=DYN_Cube5_1}
      boxExtents=<50 50 50>
      constraintSetup {angularLimitsMax=<0 0 0>, angularLimitsMin=<0 0 0>, angularTarget=<0 0 0>, linearAxesMax=<2 5 8>, linearAxesMin=<-2 1 -8>}
      localJointOffset=<0 -3 0>
     }]
    relativeSpaceBone {boneName=b_AR_Root}
   }
   n45 = Anim Dynamics - Bone: DYN_Cube2_1(ComponentPose=@n44, Alpha=0.8, ExternalForce="0,0,0") {
    angularDampingOverride=0.9
    angularSpringConstant=150
    bAngularSpring=true
    bGravityOverrideInSimSpace=true
    bLinearSpring=true
    bOverrideAngularDamping=true
    bOverrideLinearDamping=true
    gravityScale=0
    linearDampingOverride=0.8
    linearSpringConstant=30
    alphaBoolBlend {blendOption=Linear}
    boundBone {boneName=DYN_Cube2_1}
    componentLinearAccScale=<0.17 0.17 0.17>
    componentLinearVelScale=<0.8 0.8 0.6>
    physicsBodyDefinitions [{
      sphereCollisionRadius=0
      boundBone {boneName=DYN_Cube2_1}
      boxExtents=<50 50 50>
      constraintSetup {angularLimitsMax=<0 0 0>, angularLimitsMin=<0 0 0>, angularTarget=<0 0 0>, linearAxesMax=<2 5 8>, linearAxesMin=<-2 -1 -8>}
      localJointOffset=<0 -3 0>
     }]
    relativeSpaceBone {boneName=b_AR_Root}
   }
   n46 = Anim Dynamics - Bone: DYN_Cube4_1(ComponentPose=@n45, Alpha=0.8, ExternalForce="0,0,0") {
    angularDampingOverride=0.9
    angularSpringConstant=150
    bAngularSpring=true
    bGravityOverrideInSimSpace=true
    bLinearSpring=true
    bOverrideAngularDamping=true
    bOverrideLinearDamping=true
    gravityScale=0
    linearDampingOverride=0.8
    linearSpringConstant=30
    alphaBoolBlend {blendOption=Linear}
    boundBone {boneName=DYN_Cube4_1}
    componentLinearAccScale=<0.19 0.19 0.19>
    componentLinearVelScale=<0.8 0.8 0.6>
    physicsBodyDefinitions [{
      sphereCollisionRadius=0
      boundBone {boneName=DYN_Cube4_1}
      boxExtents=<50 50 50>
      constraintSetup {angularLimitsMax=<0 0 0>, angularLimitsMin=<0 0 0>, angularTarget=<0 0 0>, linearAxesMax=<2 5 8>, linearAxesMin=<-2 0 -8>}
      localJointOffset=<0 -3 0>
     }]
    relativeSpaceBone {boneName=b_AR_Root}
   }
   n47 = Anim Dynamics - Bone: DYN_Cube6_1(ComponentPose=@n46, Alpha=0.8, ExternalForce="0,0,0") {
    angularDampingOverride=0.9
    angularSpringConstant=150
    bAngularSpring=true
    bGravityOverrideInSimSpace=true
    bLinearSpring=true
    bOverrideAngularDamping=true
    bOverrideLinearDamping=true
    gravityScale=0
    linearDampingOverride=0.8
    linearSpringConstant=30
    alphaBoolBlend {blendOption=Linear}
    boundBone {boneName=DYN_Cube6_1}
    componentLinearAccScale=<0.2 0.2 0.2>
    componentLinearVelScale=<0.8 1 0.6>
    physicsBodyDefinitions [{
      sphereCollisionRadius=0
      boundBone {boneName=DYN_Cube6_1}
      boxExtents=<50 50 50>
      constraintSetup {angularLimitsMax=<0 0 0>, angularLimitsMin=<0 0 0>, angularTarget=<0 0 0>, linearAxesMax=<2 5 8>, linearAxesMin=<-2 1 -8>}
      localJointOffset=<0 -3 0>
     }]
    relativeSpaceBone {boneName=b_AR_Root}
   }
   n48 = Component To Local(ComponentPose=@n47)
   n49 = Get bIsPlayingMeleeAnim()
   n50 = Anim Dynamics - Bone: DYN_Cube1_1(
    ComponentPose=Local To Component(LocalPose=Copy Pose From Mesh() {bCopyCurves=true, bUseAttachedParent=true})
    Alpha=1
    ExternalForce="0,0,0"
   ) {
    angularDampingOverride=0.9
    bAngularSpring=true
    bGravityOverrideInSimSpace=true
    bLinearSpring=true
    bOverrideAngularBias=true
    bOverrideAngularDamping=true
    bOverrideLinearDamping=true
    gravityScale=0
    linearDampingOverride=0.9
    linearSpringConstant=10
    alphaBoolBlend {blendOption=Linear}
    boundBone {boneName=DYN_Cube1_1}
    componentLinearAccScale=<0.2 0.2 0.2>
    componentLinearVelScale=<0.7 0.7 0.7>
    physicsBodyDefinitions [{
      sphereCollisionRadius=0
      boundBone {boneName=DYN_Cube1_1}
      boxExtents=<50 50 50>
      constraintSetup {angularLimitsMax=<0 0 0>, angularLimitsMin=<0 0 0>, angularTarget=<0 0 0>, linearAxesMax=<0 2 8>, linearAxesMin=<0 -2 -8>}
      localJointOffset=<-3 0 0>
     }]
    relativeSpaceBone {boneName=b_AR_Root}
   }
   n51 = Anim Dynamics - Bone: DYN_Cube3_1(ComponentPose=@n50, Alpha=1, ExternalForce="0,0,0") {
    angularDampingOverride=0.9
    bAngularSpring=true
    bGravityOverrideInSimSpace=true
    bLinearSpring=true
    bOverrideAngularBias=true
    bOverrideAngularDamping=true
    bOverrideLinearDamping=true
    gravityScale=0
    linearDampingOverride=0.9
    linearSpringConstant=10
    alphaBoolBlend {blendOption=Linear}
    boundBone {boneName=DYN_Cube3_1}
    componentLinearAccScale=<0.2 0.2 0.2>
    componentLinearVelScale=<0.7 0.7 0.7>
    physicsBodyDefinitions [{
      sphereCollisionRadius=0
      boundBone {boneName=DYN_Cube3_1}
      boxExtents=<50 50 50>
      constraintSetup {angularLimitsMax=<0 0 0>, angularLimitsMin=<0 0 0>, angularTarget=<0 0 0>, linearAxesMax=<0 2 8>, linearAxesMin=<0 -2 -8>}
      localJointOffset=<-3 0 0>
     }]
    relativeSpaceBone {boneName=b_AR_Root}
   }
   n52 = Anim Dynamics - Bone: DYN_Cube5_1(ComponentPose=@n51, Alpha=1, ExternalForce="0,0,0") {
    angularDampingOverride=0.9
    bAngularSpring=true
    bGravityOverrideInSimSpace=true
    bLinearSpring=true
    bOverrideAngularBias=true
    bOverrideAngularDamping=true
    bOverrideLinearDamping=true
    gravityScale=0
    linearDampingOverride=0.9
    linearSpringConstant=10
    alphaBoolBlend {blendOption=Linear}
    boundBone {boneName=DYN_Cube5_1}
    componentLinearAccScale=<0.2 0.2 0.2>
    componentLinearVelScale=<0.7 0.7 0.7>
    physicsBodyDefinitions [{
      sphereCollisionRadius=0
      boundBone {boneName=DYN_Cube5_1}
      boxExtents=<50 50 50>
      constraintSetup {angularLimitsMax=<0 0 0>, angularLimitsMin=<0 0 0>, angularTarget=<0 0 0>, linearAxesMax=<0 2 8>, linearAxesMin=<0 -2 -8>}
      localJointOffset=<-3 0 0>
     }]
    relativeSpaceBone {boneName=b_AR_Root}
   }
   n53 = Anim Dynamics - Bone: DYN_Cube1_3(ComponentPose=@n52, Alpha=1, ExternalForce="0,0,0") {
    angularDampingOverride=0.9
    bAngularSpring=true
    bGravityOverrideInSimSpace=true
    bLinearSpring=true
    bOverrideAngularBias=true
    bOverrideAngularDamping=true
    bOverrideLinearDamping=true
    gravityScale=0
    linearDampingOverride=0.9
    linearSpringConstant=10
    alphaBoolBlend {blendOption=Linear}
    boundBone {boneName=DYN_Cube1_3}
    componentLinearAccScale=<0.2 0.2 0.2>
    componentLinearVelScale=<0.7 0.7 0.7>
    physicsBodyDefinitions [{
      sphereCollisionRadius=0
      boundBone {boneName=DYN_Cube1_3}
      boxExtents=<50 50 50>
      constraintSetup {angularLimitsMax=<0 0 0>, angularLimitsMin=<0 0 0>, angularTarget=<0 0 0>, linearAxesMax=<0 2 8>, linearAxesMin=<0 -2 -8>}
      localJointOffset=<-3 0 0>
     }]
    relativeSpaceBone {boneName=b_AR_Root}
   }
   n54 = Anim Dynamics - Bone: DYN_Cube2_1(ComponentPose=@n53, Alpha=1, ExternalForce="0,0,0") {
    angularDampingOverride=0.9
    bAngularSpring=true
    bGravityOverrideInSimSpace=true
    bLinearSpring=true
    bOverrideAngularBias=true
    bOverrideAngularDamping=true
    bOverrideLinearDamping=true
    gravityScale=0
    linearDampingOverride=0.9
    linearSpringConstant=10
    alphaBoolBlend {blendOption=Linear}
    boundBone {boneName=DYN_Cube2_1}
    componentLinearAccScale=<0.2 0.2 0.2>
    componentLinearVelScale=<0.7 0.7 0.7>
    physicsBodyDefinitions [{
      sphereCollisionRadius=0
      boundBone {boneName=DYN_Cube2_1}
      boxExtents=<50 50 50>
      constraintSetup {angularLimitsMax=<0 0 0>, angularLimitsMin=<0 0 0>, angularTarget=<0 0 0>, linearAxesMax=<0 2 8>, linearAxesMin=<0 -2 -8>}
      localJointOffset=<-3 0 0>
     }]
    relativeSpaceBone {boneName=b_AR_Root}
   }
   n55 = Anim Dynamics - Bone: DYN_Cube4_1(ComponentPose=@n54, Alpha=1, ExternalForce="0,0,0") {
    angularDampingOverride=0.9
    bAngularSpring=true
    bGravityOverrideInSimSpace=true
    bLinearSpring=true
    bOverrideAngularBias=true
    bOverrideAngularDamping=true
    bOverrideLinearDamping=true
    gravityScale=0
    linearDampingOverride=0.9
    linearSpringConstant=10
    alphaBoolBlend {blendOption=Linear}
    alphaScaleBiasClamp {scale=0.5}
    boundBone {boneName=DYN_Cube4_1}
    componentLinearAccScale=<0.2 0.2 0.2>
    componentLinearVelScale=<0.7 0.7 0.7>
    physicsBodyDefinitions [{
      sphereCollisionRadius=0
      boundBone {boneName=DYN_Cube4_1}
      boxExtents=<50 50 50>
      constraintSetup {angularLimitsMax=<0 0 0>, angularLimitsMin=<0 0 0>, angularTarget=<0 0 0>, linearAxesMax=<0 2 8>, linearAxesMin=<0 -2 -8>}
      localJointOffset=<-3 0 0>
     }]
    relativeSpaceBone {boneName=b_AR_Root}
   }
   n56 = Anim Dynamics - Bone: DYN_Cube6_1(ComponentPose=@n55, Alpha=1, ExternalForce="0,0,0") {
    angularDampingOverride=0.9
    bAngularSpring=true
    bGravityOverrideInSimSpace=true
    bLinearSpring=true
    bOverrideAngularBias=true
    bOverrideAngularDamping=true
    bOverrideLinearDamping=true
    gravityScale=0
    linearDampingOverride=0.9
    linearSpringConstant=10
    alphaBoolBlend {blendOption=Linear}
    boundBone {boneName=DYN_Cube6_1}
    componentLinearAccScale=<0.2 0.2 0.2>
    componentLinearVelScale=<0.7 0.7 0.7>
    physicsBodyDefinitions [{
      sphereCollisionRadius=0
      boundBone {boneName=DYN_Cube6_1}
      boxExtents=<50 50 50>
      constraintSetup {angularLimitsMax=<0 0 0>, angularLimitsMin=<0 0 0>, angularTarget=<0 0 0>, linearAxesMax=<0 2 8>, linearAxesMin=<0 -2 -8>}
      localJointOffset=<-3 0 0>
     }]
    relativeSpaceBone {boneName=b_AR_Root}
   }
   n57 = Anim Dynamics - Bone: DYN_Base(ComponentPose=@n56, Alpha=1, ExternalForce="0,0,0") {
    angularDampingOverride=0.9
    angularSpringConstant=150
    bAngularSpring=true
    bGravityOverrideInSimSpace=true
    bLinearSpring=true
    bOverrideAngularDamping=true
    bOverrideLinearDamping=true
    gravityScale=0
    linearDampingOverride=0.995
    linearSpringConstant=60
    simulationSpace=RootRelative
    alphaBoolBlend {blendOption=Linear}
    alphaScaleBiasClamp {scale=0.3}
    boundBone {boneName=DYN_Base}
    componentLinearAccScale=<0.3 0.3 0.3>
    componentLinearVelScale=<0.6 0.6 0.6>
    gravityOverride {z=-980}
    physicsBodyDefinitions [{
      sphereCollisionRadius=0
      boundBone {boneName=DYN_Base}
      boxExtents=<50 50 50>
      constraintSetup {angularLimitsMax=<0 0 0>, angularLimitsMin=<0 0 0>, angularTarget=<0 0 0>, linearAxesMax=<4 4 4>, linearAxesMin=<-4 -4 -4>}
      localJointOffset=<-3 0 0>
     }]
    relativeSpaceBone {boneName=b_AR_Root}
   }
   n58 = Get Cube_Float_A()
   n59 = Get Floating_Rotator2()
   n60 = "Transform (Modify) Bone - Bone: DYN_Cube1_2"(ComponentPose=@n57, Alpha=1, Translation=@n58, Rotation=@n59, Scale="1,1,1") {rotationMode=BMM_Additive, translationMode=BMM_Additive, alphaBoolBlend {blendOption=Linear}, boneToModify {boneName=DYN_Cube1_2}}
   n61 = Get Cube_Float_C()
   n62 = Get Floating_Rotator1()
   n63 = "Transform (Modify) Bone - Bone: DYN_Cube2_2"(ComponentPose=@n60, Alpha=1, Translation=@n61, Rotation=@n62, Scale="1,1,1") {rotationMode=BMM_Additive, translationMode=BMM_Additive, alphaBoolBlend {blendOption=Linear}, boneToModify {boneName=DYN_Cube2_2}}
   n64 = "Transform (Modify) Bone - Bone: DYN_Cube3_2"(ComponentPose=@n63, Alpha=0.5, Translation=Get Cube_Float_B(), Rotation=@n59, Scale="1,1,1") {rotationMode=BMM_Additive, translationMode=BMM_Additive, alphaBoolBlend {blendOption=Linear}, boneToModify {boneName=DYN_Cube3_2}}
   n65 = "Transform (Modify) Bone - Bone: DYN_Cube4_2"(ComponentPose=@n64, Alpha=1, Translation=Get Cube_Float_D(), Rotation=@n62, Scale="1,1,1") {rotationMode=BMM_Additive, translationMode=BMM_Additive, alphaBoolBlend {blendOption=Linear}, boneToModify {boneName=DYN_Cube4_2}}
   n66 = "Transform (Modify) Bone - Bone: DYN_Cube5_2"(ComponentPose=@n65, Alpha=1, Translation=@n61, Rotation=@n62, Scale="1,1,1") {rotationMode=BMM_Additive, translationMode=BMM_Additive, alphaBoolBlend {blendOption=Linear}, boneToModify {boneName=DYN_Cube5_2}}
   n67 = "Transform (Modify) Bone - Bone: DYN_Cube6_2"(ComponentPose=@n66, Alpha=0.5, Translation=@n58, Rotation=@n59, Scale="1,1,1") {rotationMode=BMM_Additive, translationMode=BMM_Additive, alphaBoolBlend {blendOption=Linear}, boneToModify {boneName=DYN_Cube6_2}}
   n68 = "Transform (Modify) Bone - Bone: DYN_Cube1_4"(ComponentPose=@n67, Alpha=1, Translation=@n61, Rotation=@n62, Scale="1,1,1") {rotationMode=BMM_Additive, translationMode=BMM_Additive, alphaBoolBlend {blendOption=Linear}, boneToModify {boneName=DYN_Cube1_4}}
   n69 = "Transform (Modify) Bone - Bone: DYN_Base"(ComponentPose=@n68, Alpha=1, Translation=@n58, Rotation=Get Base_Rotator() -> (Base_Rotator:Rotator), Scale="1,1,1") {
    rotationMode=BMM_Additive
    rotationSpace=BCS_BoneSpace
    translationMode=BMM_Additive
    translationSpace=BCS_BoneSpace
    alphaBoolBlend {blendOption=Linear}
    boneToModify {boneName=DYN_Base}
   }
   n70 = "Transform (Modify) Bone - Bone: DYN_MSTR_EnergyPlane"(ComponentPose=@n69, Alpha=1, Translation="1,0.7,0", Rotation="0,0,0", Scale="1,1,1") {
    rotationMode=BMM_Additive
    rotationSpace=BCS_BoneSpace
    translationMode=BMM_Additive
    translationSpace=BCS_BoneSpace
    alphaBoolBlend {blendOption=Linear}
    boneToModify {boneName=DYN_MSTR_EnergyPlane}
    translation=<1 0.7>
   }
   Output Pose(Result=Blend Poses by bool(
     BlendPose_0=Component To Local(ComponentPose=@n70)
     BlendPose_1=Blend Poses by bool(BlendPose_0=@n48, BlendPose_1=@n41, BlendTime_0=1, BlendTime_1=1, bActiveValue=@n49) {blendType=Linear}
     bActiveValue=Get bIsInFrontEnd()
    ) {blendType=Linear}) {layerGroup=None}
   n71 = Get Cube_Float_A()
   n72 = Get Floating_Rotator2()
   n73 = "Transform (Modify) Bone - Bone: DYN_Cube1_2"(
    ComponentPose=Local To Component(LocalPose=Copy Pose From Mesh() {bCopyCurves=true, bUseAttachedParent=true})
    Alpha=1
    Translation=@n71
    Rotation=@n72
    Scale="1,1,1"
   ) {rotationMode=BMM_Additive, translationMode=BMM_Additive, alphaBoolBlend {blendOption=Linear}, boneToModify {boneName=DYN_Cube1_2}}
   n74 = Get Cube_Float_C()
   n75 = Get Floating_Rotator1()
   n76 = "Transform (Modify) Bone - Bone: DYN_Cube1_4"(ComponentPose=@n73, Alpha=1, Translation=@n74, Rotation=@n75, Scale="1,1,1") {rotationMode=BMM_Additive, translationMode=BMM_Additive, alphaBoolBlend {blendOption=Linear}, boneToModify {boneName=DYN_Cube1_4}}
   n77 = "Transform (Modify) Bone - Bone: DYN_Base"(ComponentPose=@n76, Alpha=0.7, Translation=@n71, Rotation=@n75, Scale="1,1,1") {
    rotationMode=BMM_Additive
    translationMode=BMM_Additive
    translationSpace=BCS_BoneSpace
    alphaBoolBlend {blendOption=Linear}
    boneToModify {boneName=DYN_Base}
   }
   n78 = "Transform (Modify) Bone - Bone: DYN_Cube3_2"(ComponentPose=@n77, Alpha=0.5, Translation=Get Cube_Float_B(), Rotation=@n72, Scale="1,1,1") {rotationMode=BMM_Additive, translationMode=BMM_Additive, alphaBoolBlend {blendOption=Linear}, boneToModify {boneName=DYN_Cube3_2}}
   n79 = "Transform (Modify) Bone - Bone: DYN_Cube6_2"(ComponentPose=@n78, Alpha=0.5, Translation=@n71, Rotation=@n72, Scale="1,1,1") {rotationMode=BMM_Additive, translationMode=BMM_Additive, alphaBoolBlend {blendOption=Linear}, boneToModify {boneName=DYN_Cube6_2}}
   n80 = "Transform (Modify) Bone - Bone: DYN_Cube2_2"(ComponentPose=@n79, Alpha=1, Translation=@n74, Rotation=@n75, Scale="1,1,1") {rotationMode=BMM_Additive, translationMode=BMM_Additive, alphaBoolBlend {blendOption=Linear}, boneToModify {boneName=DYN_Cube2_2}}
   n81 = "Transform (Modify) Bone - Bone: DYN_Cube4_2"(ComponentPose=@n80, Alpha=1, Translation=Get Cube_Float_D(), Rotation=@n75, Scale="1,1,1") {rotationMode=BMM_Additive, translationMode=BMM_Additive, alphaBoolBlend {blendOption=Linear}, boneToModify {boneName=DYN_Cube4_2}}
   n82 = "Transform (Modify) Bone - Bone: DYN_Cube5_2"(ComponentPose=@n81, Alpha=1, Translation=@n74, Rotation=@n75, Scale="1,1,1") {rotationMode=BMM_Additive, translationMode=BMM_Additive, alphaBoolBlend {blendOption=Linear}, boneToModify {boneName=DYN_Cube5_2}}
   n83 = Anim Dynamics - Bone: DYN_Cube1_1(ComponentPose=@n82, Alpha=0.8, ExternalForce="0,0,0") {
    angularDampingOverride=0.9
    angularSpringConstant=150
    bAngularSpring=true
    bGravityOverrideInSimSpace=true
    bLinearSpring=true
    bOverrideAngularDamping=true
    bOverrideLinearDamping=true
    gravityScale=0
    linearDampingOverride=0.9
    linearSpringConstant=10
    simulationSpace=BoneRelative
    alphaBoolBlend {blendOption=Linear}
    boundBone {boneName=DYN_Cube1_1}
    componentLinearVelScale=<0.8 0.8 0.6>
    physicsBodyDefinitions [{
      sphereCollisionRadius=0
      boundBone {boneName=DYN_Cube1_1}
      boxExtents=<50 50 50>
      constraintSetup {angularLimitsMax=<0 0 0>, angularLimitsMin=<0 0 0>, angularTarget=<0 0 0>, linearAxesMax=<2 5 8>, linearAxesMin=<-2 -1 -8>}
      localJointOffset=<0 -3 0>
     }]
    relativeSpaceBone {boneName=b_AR_Root}
   }
   n84 = Anim Dynamics - Bone: DYN_Cube3_1(ComponentPose=@n83, Alpha=0.8, ExternalForce="0,0,0") {
    angularDampingOverride=0.9
    angularSpringConstant=150
    bAngularSpring=true
    bGravityOverrideInSimSpace=true
    bLinearSpring=true
    bOverrideAngularDamping=true
    bOverrideLinearDamping=true
    gravityScale=0
    linearDampingOverride=0.9
    linearSpringConstant=10
    simulationSpace=BoneRelative
    alphaBoolBlend {blendOption=Linear}
    boundBone {boneName=DYN_Cube3_1}
    componentLinearVelScale=<0.8 0.8 0.6>
    physicsBodyDefinitions [{
      sphereCollisionRadius=0
      boundBone {boneName=DYN_Cube3_1}
      boxExtents=<50 50 50>
      constraintSetup {angularLimitsMax=<0 0 0>, angularLimitsMin=<0 0 0>, angularTarget=<0 0 0>, linearAxesMax=<2 5 8>, linearAxesMin=<-2 0 -8>}
      localJointOffset=<0 -3 0>
     }]
    relativeSpaceBone {boneName=b_AR_Root}
   }
   n85 = Anim Dynamics - Bone: DYN_Cube5_1(ComponentPose=@n84, Alpha=0.8, ExternalForce="0,0,0") {
    angularDampingOverride=0.9
    angularSpringConstant=150
    bAngularSpring=true
    bGravityOverrideInSimSpace=true
    bLinearSpring=true
    bOverrideAngularDamping=true
    bOverrideLinearDamping=true
    gravityScale=0
    linearDampingOverride=0.9
    linearSpringConstant=10
    simulationSpace=BoneRelative
    alphaBoolBlend {blendOption=Linear}
    boundBone {boneName=DYN_Cube5_1}
    componentLinearVelScale=<0.8 1 0.6>
    physicsBodyDefinitions [{
      sphereCollisionRadius=0
      boundBone {boneName=DYN_Cube5_1}
      boxExtents=<50 50 50>
      constraintSetup {angularLimitsMax=<0 0 0>, angularLimitsMin=<0 0 0>, angularTarget=<0 0 0>, linearAxesMax=<2 5 8>, linearAxesMin=<-2 1 -8>}
      localJointOffset=<0 -3 0>
     }]
    relativeSpaceBone {boneName=b_AR_Root}
   }
   n86 = Anim Dynamics - Bone: DYN_Cube2_1(ComponentPose=@n85, Alpha=0.8, ExternalForce="0,0,0") {
    angularDampingOverride=0.9
    angularSpringConstant=150
    bAngularSpring=true
    bGravityOverrideInSimSpace=true
    bLinearSpring=true
    bOverrideAngularDamping=true
    bOverrideLinearDamping=true
    gravityScale=0
    linearDampingOverride=0.9
    linearSpringConstant=10
    simulationSpace=BoneRelative
    alphaBoolBlend {blendOption=Linear}
    boundBone {boneName=DYN_Cube2_1}
    componentLinearVelScale=<0.8 0.8 0.6>
    physicsBodyDefinitions [{
      sphereCollisionRadius=0
      boundBone {boneName=DYN_Cube2_1}
      boxExtents=<50 50 50>
      constraintSetup {angularLimitsMax=<0 0 0>, angularLimitsMin=<0 0 0>, angularTarget=<0 0 0>, linearAxesMax=<2 5 8>, linearAxesMin=<-2 -1 -8>}
      localJointOffset=<0 -3 0>
     }]
    relativeSpaceBone {boneName=b_AR_Root}
   }
   n87 = Anim Dynamics - Bone: DYN_Cube4_1(ComponentPose=@n86, Alpha=0.8, ExternalForce="0,0,0") {
    angularDampingOverride=0.9
    angularSpringConstant=150
    bAngularSpring=true
    bGravityOverrideInSimSpace=true
    bLinearSpring=true
    bOverrideAngularDamping=true
    bOverrideLinearDamping=true
    gravityScale=0
    linearDampingOverride=0.9
    linearSpringConstant=10
    simulationSpace=BoneRelative
    alphaBoolBlend {blendOption=Linear}
    boundBone {boneName=DYN_Cube4_1}
    componentLinearVelScale=<0.8 0.8 0.6>
    physicsBodyDefinitions [{
      sphereCollisionRadius=0
      boundBone {boneName=DYN_Cube4_1}
      boxExtents=<50 50 50>
      constraintSetup {angularLimitsMax=<0 0 0>, angularLimitsMin=<0 0 0>, angularTarget=<0 0 0>, linearAxesMax=<2 5 8>, linearAxesMin=<-2 0 -8>}
      localJointOffset=<0 -3 0>
     }]
    relativeSpaceBone {boneName=b_AR_Root}
   }
   n88 = Anim Dynamics - Bone: DYN_Cube6_1(ComponentPose=@n87, Alpha=0.8, ExternalForce="0,0,0") {
    angularDampingOverride=0.9
    angularSpringConstant=150
    bAngularSpring=true
    bGravityOverrideInSimSpace=true
    bLinearSpring=true
    bOverrideAngularDamping=true
    bOverrideLinearDamping=true
    gravityScale=0
    linearDampingOverride=0.9
    linearSpringConstant=10
    simulationSpace=BoneRelative
    alphaBoolBlend {blendOption=Linear}
    boundBone {boneName=DYN_Cube6_1}
    componentLinearVelScale=<0.8 1 0.6>
    physicsBodyDefinitions [{
      sphereCollisionRadius=0
      boundBone {boneName=DYN_Cube6_1}
      boxExtents=<50 50 50>
      constraintSetup {angularLimitsMax=<0 0 0>, angularLimitsMin=<0 0 0>, angularTarget=<0 0 0>, linearAxesMax=<2 5 8>, linearAxesMin=<-2 1 -8>}
      localJointOffset=<0 -3 0>
     }]
    relativeSpaceBone {boneName=b_AR_Root}
   }
   n89 = Anim Dynamics - Bone: DYN_Cube1_3(ComponentPose=@n88, Alpha=0.8, ExternalForce="0,0,0") {
    angularDampingOverride=0.9
    angularSpringConstant=150
    bAngularSpring=true
    bGravityOverrideInSimSpace=true
    bLinearSpring=true
    bOverrideAngularDamping=true
    bOverrideLinearDamping=true
    gravityScale=0
    linearDampingOverride=0.9
    linearSpringConstant=10
    simulationSpace=BoneRelative
    alphaBoolBlend {blendOption=Linear}
    boundBone {boneName=DYN_Cube1_3}
    componentLinearVelScale=<0.8 0.8 0.6>
    physicsBodyDefinitions [{
      sphereCollisionRadius=0
      boundBone {boneName=DYN_Cube1_3}
      boxExtents=<50 50 50>
      constraintSetup {angularLimitsMax=<0 0 0>, angularLimitsMin=<0 0 0>, angularTarget=<0 0 0>, linearAxesMax=<2 5 8>, linearAxesMin=<-2 -1 -8>}
      localJointOffset=<0 -3 0>
     }]
    relativeSpaceBone {boneName=b_AR_Root}
   }
   n90 = Anim Dynamics - Bone: DYN_Base(ComponentPose=@n89, Alpha=0.2, ExternalForce="0,0,0") {
    angularDampingOverride=0.9
    angularSpringConstant=150
    bAngularSpring=true
    bGravityOverrideInSimSpace=true
    bLinearSpring=true
    bOverrideAngularDamping=true
    bOverrideLinearDamping=true
    gravityScale=0
    linearDampingOverride=0.995
    linearSpringConstant=60
    simulationSpace=BoneRelative
    alphaBoolBlend {blendOption=Linear}
    boundBone {boneName=DYN_Base}
    componentLinearVelScale=<0.9 0.9 0.9>
    gravityOverride {z=-980}
    physicsBodyDefinitions [{
      sphereCollisionRadius=0
      boundBone {boneName=DYN_Base}
      boxExtents=<50 50 50>
      constraintSetup {angularLimitsMax=<0 0 0>, angularLimitsMin=<0 0 0>, angularTarget=<0 0 0>, linearAxesMax=<4 4 4>, linearAxesMin=<-4 -4 -4>}
      localJointOffset=<3 0 0>
     }]
    relativeSpaceBone {boneName=b_AR_Root}
   }
   Save cached pose 'OnGroundCache'(Pose=Component To Local(ComponentPose=@n90)) {cacheName=OnGroundCache}
  ]}
 EventGraph {nodes [
   Event BlueprintUpdateAnimation()
   > n1 = Sequence!(execute) -> (then_0, then_1, then_2:Exec)
   n2 = GetGameTimeInSeconds()
   n3 = "Sin (Radians)"(A=float * float(A=float * float(A=1, B=1.5), B=@n2))
   n4 = "Sin (Radians)"(A=float * float(A=float * float(A=1, B=1), B=@n2))
   n5 = "Sin (Radians)"(A=float * float(A=float * float(A=1, B=2), B=@n2))
   Set Cube_Float_A(
    execute=@n1.then_1
    Cube_Float_A:Vector=MakeVector(X=float * float(A=@n3, B=0.3), Y=float * float(A=@n4, B=0.56), Z=float * float(A=@n5, B=0.22))
   ) -> (Output_Get:Vector)
   n6 = "Cos (Radians)"!(A:double=float * float(A=float * float(A=1, B=1.5), B=@n2)) -> (ReturnValue:double)
   n7 = "Cos (Radians)"(A=float * float(A=float * float(A=1, B=1.78), B=@n2))
   n8 = "Cos (Radians)"(A=float * float(A=float * float(A=1, B=1.15), B=@n2))
   > Set Cube_Float_B(Cube_Float_B:Vector=MakeVector(X=float * float(A=@n6, B=0.6), Y=float * float(A=@n7, B=0.725), Z=float * float(A=@n8, B=0.3))) -> (Output_Get:Vector)
   n9 = float * float(A=float * float(A=1, B=1.5), B=@n2)
   n10 = float * float(A=float * float(A=1, B=1), B=@n2)
   n11 = float * float(A=float * float(A=1, B=2), B=@n2)
   n12 = MakeVector(
    X=float * float(A="Sin (Radians)"(A=@n9), B=0.3)
    Y=float * float(A="Sin (Radians)"(A=@n10), B=0.6)
    Z=float * float(A="Sin (Radians)"(A=@n11), B=0.25)
   )
   > Set Cube_Float_C(Cube_Float_C:Vector=@n12) -> (Output_Get:Vector)
   n13 = float * float(A=float * float(A=0.6678, B=1.5), B=@n2)
   n14 = float * float(A=float * float(A=1, B=1.47), B=@n2)
   n15 = float * float(A=float * float(A=1, B=0.953), B=@n2)
   n16 = MakeVector(
    X=float * float(A="Sin (Radians)"(A=@n13), B=0.85)
    Y=float * float(A="Sin (Radians)"(A=@n15), B=0.1)
    Z=float * float(A="Sin (Radians)"(A=@n14), B=0.7)
   )
   > Set Cube_Float_D(Cube_Float_D:Vector=@n16) -> (Output_Get:Vector)
   n17 = BreakVector(InVec=vector * float(A=@n16, B=5))
   > Set Floating_Rotator2(Floating_Rotator2:Rotator=MakeRotator(Roll=@n17, Pitch=@n17.Y, Yaw=@n17.Z)) -> (Output_Get:Rotator)
   n18 = BreakVector(InVec=vector * float(A=@n12, B=7))
   > Set Floating_Rotator1(Floating_Rotator1:Rotator=MakeRotator(Roll=@n18, Pitch=@n18.Y, Yaw=@n18.Z)) -> (Output_Get:Rotator)
   n19 = Reroute Node(InputPin:double) -> (OutputPin:double)
   n20 = float * float(A=float * float(A=1, B=1), B=@n19)
   n21 = float * float(A=float * float(A=1, B=2), B=@n19)
   n22 = float * float(A=float * float(A=1, B=1.5), B=@n19)
   n23 = MakeVector(
    X=float * float(A="Sin (Radians)"(A=@n22), B=0.2)
    Y=float * float(A="Sin (Radians)"(A=@n20), B=0.45)
    Z=float * float(A="Sin (Radians)"(A=@n21), B=0.15)
   )
   > Set BaseFloater(BaseFloater:Vector=@n23) -> (Output_Get:Vector)
   n24 = BreakVector(InVec=vector * float(A=@n23, B=4))
   > Set Base_Rotator(Base_Rotator:Rotator=MakeRotator(Roll=@n24, Pitch=@n24.Y, Yaw=@n24.Z)) -> (Output_Get:Rotator)
  ]}
]
variables [
 Base_Rotator {type=Rotator}
 BaseFloater {type=Vector}
 Cube_Float_A {type=Vector}
 Cube_Float_B {type=Vector}
 Cube_Float_C {type=Vector}
 Cube_Float_D {type=Vector}
 Floating_Rotator1 {type=Rotator}
 Floating_Rotator2 {type=Rotator}
]
